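# Niagara burst effect for the MistRaven emote's book moment, mixing dark and red smoke, tumbling feathers, a brief flare and a flickering red glow.
# Smoke and feathers rise and slow, with curl noise scattering the feathers.
# Eyeglow layers sprites, mesh and light, scaled by time of day and front-end context, plus platform for some layers.
# Wave, Eyeglow001 and the Ascended-driven smoke-light colour setter are disabled.
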NiagaraSystem /BRCosmetics/Animation/Game/MainPlayer/Emotes/MistRaven/FX/NS_Emote_MistRaven_Book_01
dependencies {
 usedDataInterfaces [
  &NiagaraDataInterfaceCurve
  &NiagaraDataInterfaceVectorCurve
  &NiagaraDataInterfaceVector2DCurve
  &NiagaraDataInterfaceSpriteRendererInfo
  &NiagaraDataInterfaceColorCurve
  &NiagaraDataInterfacePlatformSet
  &NiagaraDataInterfaceSkeletalMesh
 ]
 usedDynamicInputs [
  &Vector2DFromCurve
  &RandomRangeVector2D
  &MakeVector4_FromLinearColor
 ]
 usedModules [
  &self:SystemSpawnScript.SystemScriptSource.SystemScriptGraph.NiagaraNodeAssignment_0.SetVariables_#1
  &SystemState
 ]
}
emitters {
 Em_Smoke_Cloud {
  emitterUpdateScript [
   &EmitterState(
    Inactive Response=::NewEnumerator1
    Life Cycle Mode=::NewEnumerator1
    Loop Behavior=::NewEnumerator1
    Loop Delay=1.9
    Loop Duration=2
    UseLoopDelay=false
   )
   &SpawnBurst_Instantaneous(Spawn Count=4, Use Spawn Probability=false, UseLoopCountLimit=false)
   /SetVariables(
    Emitter.Lit Color:LinearColor=&MakeLinearColorFromVectorAndFloat(
     "Vector (RGB)"=&MakeVectorFromLinearColorRGB(Color=&MakeCustomLinearColorFromBool(Bool=$User.Ascended, True Color=<rgba 3.2915 0 20 1>, False Color=<rgba 20 0 0.1747 1>))
     "Float (Alpha)"=&FloatFromCurve(
      FloatCurve=DataInterface() {propertyValues {Curve {keys [{arriveTangent=-2.6103, interpMode=RCIM_Cubic, leaveTangent=-2.6103, tangentMode=RCTM_User, time=0, value=1}, <at 0.3 0 RCIM_Cubic>]}}}
      CurveIndex=$Emitter.Age
     )
    )
    Emitter.Lit Position:NiagaraPosition=<0 0 80>
    Emitter.Lit Radius:NiagaraFloat=&FloatFromCurve(
     FloatCurve=DataInterface() {propertyValues {Curve {keys [{arriveTangent=-2.6103, interpMode=RCIM_Cubic, leaveTangent=-2.6103, tangentMode=RCTM_User, time=0.1, value=1}, <at 0.3 0 RCIM_Cubic>]}}}
     CurveIndex=$Emitter.Age
     Scale Curve=100
    )
   ) {enabled=false, isSetParameters=true}
  ]
  particleSpawnScript [
   &InitializeParticle(
    Color=&MergeHelper_ColorSwitch(
     0=<rgba 0.0118 0.0108 0.0273 1>
     1=<rgba 0.401 0.0126 0.0125 2>
     2=?
     3=?
     Index=&RandomRangeInt(Minimum=0, Maximum=1)
     Num=::NewEnumerator0
    )
    Lifetime Max=1.15
    Lifetime Min=0.9
    Lifetime Mode=::NewEnumerator1
    Sprite Rotation Mode=::NewEnumerator2
    Sprite Size Mode=::NewEnumerator1
    Uniform Sprite Size Max=100
    Uniform Sprite Size Min=75
    UsePositionOffset=false
    Write Material Random=true
   )
   &AddVelocity(
    Rotation Coordinate Space=::World
    Velocity=&RandomRangeVector(Minimum=<-5 -5 70>, Maximum=<5 5 90>)
    Velocity Speed Scale=1.6
   )
   &ShapeLocation(MaskSpawnGroup=false, Override Seed=false, Sphere Radius=30, UseOwnerScale=true)
  ]
  particleUpdateScript [
   &ParticleState()
   &ScaleVelocity(Velocity Scale=&VectorFromCurve(VectorCurve=DataInterface() {propertyValues {
       XCurve {keys [{arriveTangent=-1.756, interpMode=RCIM_Cubic, leaveTangent=-1.756, tangentMode=RCTM_User, time=0, value=1}, <at 1 0>]}
       YCurve {keys [{arriveTangent=-1.756, interpMode=RCIM_Cubic, leaveTangent=-1.756, tangentMode=RCTM_User, time=0, value=1}, <at 1 0>]}
       ZCurve {keys [{arriveTangent=-1.756, interpMode=RCIM_Cubic, leaveTangent=-1.756, tangentMode=RCTM_User, time=0, value=1}, <at 1 0>]}
      }}))
   &SpriteRotationRate(Rotation Rate=&RandomRangeFloat(Minimum=-150, Maximum=150))
   &ScaleSpriteSize(
    Uniform Curve Scale=0.3
    Uniform Curve Sprite Scale=DataInterface() {propertyValues {Curve {
       defaultValue=1
       keys [
        {arriveTangent=4.1785, interpMode=RCIM_Cubic, leaveTangent=4.1785, tangentMode=RCTM_User, time=0, value=1}
        {arriveTangent=1.1179, interpMode=RCIM_Cubic, leaveTangent=1.1179, tangentMode=RCTM_User, time=0.1, value=1.875}
        {arriveTangent=0.0224, interpMode=RCIM_Cubic, leaveTangent=0.0224, tangentMode=RCTM_User, time=1, value=2}
       ]
      }}}
   )
   &SolveForcesAndVelocity(Clamp Velocity=false, Limit Acceleration=false)
   &ScaleColor(
    Scale Alpha=&FloatFromCurve(FloatCurve=DataInterface() {propertyValues {Curve {keys [
         {interpMode=RCIM_Cubic, tangentMode=RCTM_User, time=0, value=0.4}
         {arriveTangent=-2.0189, interpMode=RCIM_Cubic, leaveTangent=-2.1015, tangentMode=RCTM_Break, time=0.2, value=1.35}
         {arriveTangent=-1.1521, interpMode=RCIM_Cubic, tangentMode=RCTM_Break, time=1, value=0}
        ]}}})
    ScaleA=true
    ScaleRGB=true
   )
  ]
  rendererClasses [&NiagaraSpriteRendererProperties, &NiagaraLightRendererProperties]
  renderers [
   {rendererClass=&NiagaraSpriteRendererProperties}
   {rendererClass=&NiagaraLightRendererProperties, rendererIndex=1}
  ]
 }
 Em_Smoke_Cloud_Detail {
  emitterUpdateScript [
   &EmitterState(
    Inactive Response=::NewEnumerator1
    Life Cycle Mode=::NewEnumerator1
    Loop Behavior=::NewEnumerator1
    Loop Delay=0.4
    Loop Duration=2
    UseLoopDelay=false
   )
   &SpawnBurst_Instantaneous(Spawn Count=5, Use Spawn Probability=false, UseLoopCountLimit=false)
  ]
  particleSpawnScript [
   &InitializeParticle(
    Color=&MergeHelper_ColorSwitch(
     0=<rgba 0.0118 0.0108 0.0273 1>
     1=<rgba 0.1198 0.0038 0.0037 1>
     2=?
     3=?
     Index=&RandomRangeInt(Minimum=0, Maximum=1)
     Num=::NewEnumerator0
    )
    Lifetime Max=1
    Lifetime Min=0.8
    Lifetime Mode=::NewEnumerator1
    Sprite Rotation Mode=::NewEnumerator2
    Sprite Size Max=<100 100>
    Sprite Size Min=<80 80>
    Sprite Size Mode=::NewEnumerator2
    UsePositionOffset=true
    Write Material Random=true
   )
   &CameraOffset(Camera Offset Amount=30)
   &AddVelocity(
    Module.Use Falloff Distance=false
    Origin Offset=$StackContext.ShapeLocation.ShapePosition
    Override Seed=false
    Velocity Mode=::NewEnumerator1
    Velocity Speed=100
   ) {enabled=false}
   &AddVelocity(
    Rotation Coordinate Space=::World
    Velocity=&RandomRangeVector(Minimum=<-5 -5 70>, Maximum=<5 5 90>)
    Velocity Speed Scale=1.5
   ) {module=AddVelocity001}
   &ShapeLocation(MaskSpawnGroup=false, Override Seed=false, Sphere Radius=30, UseOwnerScale=true)
  ]
  particleUpdateScript [
   &ParticleState()
   &ScaleVelocity(Velocity Scale=&VectorFromCurve(VectorCurve=DataInterface() {propertyValues {
       XCurve {keys [{arriveTangent=-1.756, interpMode=RCIM_Cubic, leaveTangent=-1.756, tangentMode=RCTM_User, time=0, value=1}, <at 1 0>]}
       YCurve {keys [{arriveTangent=-1.756, interpMode=RCIM_Cubic, leaveTangent=-1.756, tangentMode=RCTM_User, time=0, value=1}, <at 1 0>]}
       ZCurve {keys [{arriveTangent=-1.756, interpMode=RCIM_Cubic, leaveTangent=-1.756, tangentMode=RCTM_User, time=0, value=1}, <at 1 0>]}
      }}))
   &SpriteRotationRate(Rotation Rate=&RandomRangeFloat(Minimum=-50, Maximum=50))
   &ScaleSpriteSize(
    Scale Sprite Size Mode=::NewEnumerator0
    Uniform Scale Factor=&FloatFromCurve(
     FloatCurve=DataInterface() {propertyValues {Curve {
        defaultValue=1
        keys [
         {arriveTangent=4.1785, interpMode=RCIM_Cubic, leaveTangent=4.1785, tangentMode=RCTM_User, time=0, value=1}
         {arriveTangent=1.1179, interpMode=RCIM_Cubic, leaveTangent=1.1179, tangentMode=RCTM_User, time=0.2042, value=1.6}
         {arriveTangent=0.0224, interpMode=RCIM_Cubic, leaveTangent=0.0224, tangentMode=RCTM_User, time=1, value=2}
        ]
       }}}
     Scale Curve=0.38
    )
   )
   &SubUVAnimation(
    End Frame Range Override=30
    "Sprite Renderer "=DataInterface() {propertyValues {SpriteRenderer=&self:Em_Smoke_Cloud_Detail_0.NiagaraSpriteRendererProperties_0}}
    Start Frame Range Override=&RandomRangeInt(Minimum=0, Maximum=5)
    UseEndFrame=true
    UseStartFrame=true
   )
   &ScaleColor(
    Scale Alpha=&Multiply_Float(
     A=&RandomRangeFloat(Minimum=0.5, Maximum=1)
     B=&FloatFromCurve(
      FloatCurve=DataInterface() {propertyValues {Curve {keys [
          {arriveTangent=3.7485, interpMode=RCIM_Cubic, leaveTangent=3.7485, tangentMode=RCTM_User, time=0, value=0.35}
          {interpMode=RCIM_Cubic, tangentMode=RCTM_User, time=0.5, value=1}
          {tangentMode=RCTM_User, time=1, value=0}
         ]}}}
      Scale Curve=2
     )
    )
    ScaleA=true
    ScaleRGB=true
   )
   &Drag(Drag=6, Use Linear Drag=true, Use Rotational Drag=true) {enabled=false}
   &SolveForcesAndVelocity(Clamp Velocity=false, Limit Acceleration=false)
  ]
  rendererClasses [&NiagaraSpriteRendererProperties]
  renderers [{rendererClass=&NiagaraSpriteRendererProperties}]
 }
 Eyeglow {
  emitterUpdateScript [
   &EmitterState(Inactive Response=::NewEnumerator1, Life Cycle Mode=::NewEnumerator1, Loop Behavior=::NewEnumerator1, Loop Duration=2, UseLoopDelay=false)
   &SpawnBurst_Instantaneous(Use Spawn Probability=false, UseLoopCountLimit=false)
  ]
  particleSpawnScript [
   &InitializeParticle(
    Color=<rgba 0.15 0.15 0.15 1.5>
    Lifetime=0.25
    Mesh Scale=<0.08 0.08 0.08>
    Mesh Scale Mode=::NewEnumerator4
    Sprite Size=<160 160>
    Sprite Size Mode=::NewEnumerator4
    UsePositionOffset=false
    Write Material Random=true
   )
   &InitialMeshOrientation(Mesh Orientation Mode=::NewEnumerator2, Orientation Vector=<0 0 1>, Rotation=<0 0.25 0>, Use Rotation Vector=true)
   &CameraOffset(Camera Offset Amount=5)
   &AddVelocity(Velocity=<0 0 0.001>)
  ]
  particleUpdateScript [
   &ParticleState()
   &ScaleMeshSize(Scale Factor=&Multiply_Vector(
     A=&ScaleFloatByCurve(
      InitialValue:Vector3f=<1 1 1>
      ScaleCurve:NiagaraDataInterfaceVectorCurve=DataInterface() {propertyValues {
        XCurve {keys [<at 0 0.25>,<at 0.1 1>,<at 0.9 1>,<at 1 0>]}
        YCurve {keys [<at 0 1.25>,<at 0.1 2.25>,<at 0.9 2.25>,<at 1 0>]}
        ZCurve {keys [<at 0 1.25>,<at 0.1 2.25>,<at 0.9 2.25>,<at 1 0>]}
       }}
      CurveIndex:NiagaraFloat=$Particles.NormalizedAge
     )
     B=&Divide_Vector(A:Vector3f=<1 1 1>, B:Vector3f=$Engine.Owner.Scale)
    ))
   &ScaleSpriteSize(Uniform Curve Sprite Scale=DataInterface() {propertyValues {Curve {keys [
        <at 0 0>
        <at 0.1 1>
        <at 0.1512 0.7174>
        <at 0.2132 1>
        <at 0.2641 0.7539>
        <at 0.299 1>
        <at 0.3563 0.7037>
        <at 0.4151 1>
        <at 0.4564 0.8017>
        <at 0.5041 1>
        <at 0.5502 0.7561>
        <at 0.5836 1>
        <at 0.6551 0.7653>
        <at 0.6981 1>
        <at 0.7521 0.7721>
        <at 0.7871 1>
        <at 0.8475 0.7106>
        <at 0.9 1>
        {tangentMode=RCTM_User, time=1, value=0}
       ]}}})
   /SetVariables(Particles.Size002:Vector2f=&Multiply_Vector2D(A=$Particles.SpriteSize, B=<0.6 0.9>)) {isSetParameters=true, module=SetVariables_#2}
   /SetVariables(Particles.Size003:Vector2f=&Multiply_Vector2D(A=<150 50>, B=<1 3>)) {isSetParameters=true, module=SetVariables_#3}
   &ScaleColor(
    Scale Alpha=&Multiply_Float(
     A=&FloatFromCurve(
      FloatCurve=DataInterface() {propertyValues {Curve {keys [
          {arriveTangent=4.4475, time=0, value=0}
          {arriveTangent=5.3518, leaveTangent=-3.1774, time=0.1, value=1.2}
          {arriveTangent=-1.4972, leaveTangent=-1.4972, time=0.1523, value=0.7}
          {arriveTangent=-0.1667, leaveTangent=-0.1667, time=0.2052, value=1.1711}
          <at 0.258 0.7713>
          <at 0.2926 1.1775>
          {arriveTangent=-0.25, leaveTangent=-0.25, time=0.3608, value=0.689}
          <at 0.3865 1.193>
          <at 0.4263 0.8642>
          <at 0.4931 1.1191>
          {arriveTangent=-0.25, leaveTangent=-0.25, time=0.5428, value=0.7548}
          <at 0.5846 1.1063>
          <at 0.6571 0.8076>
          <at 0.6976 1.1685>
          {arriveTangent=0.1667, leaveTangent=0.1667, time=0.7542, value=0.7713}
          <at 0.7906 1.1138>
          <at 0.8535 0.8807>
          {arriveTangent=-3.3333, leaveTangent=-3.3333, time=0.9, value=1.2}
          <at 1 0>
         ]}}}
      Scale Curve=&FNTimeOfDayScalar(Night Scale=0.6, Morning Scale=0.7, Day Scale=0.8, Evening Scale=0.6)
     )
     B=&FN_FrontEndFloatModifier(Gameplay=1, Front End=0.3)
    )
    Scale RGB=&MakeVectorFromLinearColorRGB(Color=&ColorFromCurve(ColorCurve=DataInterface() {propertyValues {
        AlphaCurve {keys []}
        BlueCurve {keys [<at 0.7887 0 RCIM_Cubic>,<at 0.9801 0.0012 RCIM_Cubic>]}
        GreenCurve {keys [<at 0.7887 0.0063 RCIM_Cubic>,<at 0.9801 0 RCIM_Cubic>]}
        RedCurve {keys [<at 0.7887 0.4375 RCIM_Cubic>,<at 0.9801 0.1562 RCIM_Cubic>]}
       }}))
    ScaleA=true
    ScaleRGB=true
   )
   /SetVariables(Particles.color002:LinearColor=&MakeLinearColorFromVectorAndFloat(
     "Vector (RGB)"=&MakeVectorFromLinearColorRGB(Color=<rgba 3 0.12 0.18 1>)
     "Float (Alpha)"=&FloatFromCurve(
      FloatCurve=DataInterface() {propertyValues {Curve {keys [<at 0 1>,<at 0.1 2>,<at 0.2005 1.7767>,<at 0.306 1.9927>,<at 0.3889 1.7548>,<at 0.5026 2.0076>,<at 0.5773 1.7477>,<at 0.6601 1.9791>,<at 0.7283 1.7134>,<at 0.8209 1.9848>,<at 1 0>]}}}
      Scale Curve=&FN_PickFloatDependOnPlatform(TargetPlatformActive=5, TargetPlatformNoActive=1)
     )
    )) {isSetParameters=true, module=SetVariables_#4}
   /SetVariables(Particles.Color003:LinearColor=&MakeLinearColorFromVectorAndFloat(
     "Vector (RGB)"=&MakeVectorFromLinearColorRGB(Color=<rgba 1 0.04 0.05 1>)
     "Float (Alpha)"=&Multiply_Float(
      A=&FloatFromCurve(
       FloatCurve=DataInterface() {propertyValues {Curve {keys [<at 0 0.8>,<at 0.1 1>,<at 0.2 0.6>,<at 0.3 1>,<at 0.4 0.7>,<at 0.5 0.95>,<at 0.6 0.55>,<at 0.7 1>,<at 0.8 0.55>,<at 0.9 0.8>,<at 1 0>]}}}
       Scale Curve=&FNTimeOfDayScalar(Night Scale=0.2, Morning Scale=0.1, Day Scale=0.4, Evening Scale=0.2)
      )
      B=&Multiply_Float(
       A=&FN_FrontEndFloatModifier(Gameplay=0.5, Front End=0.15)
       B=&FN_PickFloatDependOnPlatform(TargetPlatformActive=1, TargetPlatformNoActive=0.4)
      )
     )
    )) {isSetParameters=true, module=SetVariables_#5}
   /SetVariables(Particles.MeshColor:LinearColor=&MakeLinearColorFromVectorAndFloat(
     "Vector (RGB)"=&MakeVectorFromLinearColorRGB(Color=&ColorFromCurve(ColorCurve=DataInterface() {propertyValues {
         AlphaCurve {keys []}
         BlueCurve {keys [
           {arriveTangent=-1.8762, interpMode=RCIM_Cubic, leaveTangent=-1.8762, tangentMode=RCTM_User, time=0.1195, value=0.3256}
           {interpMode=RCIM_Cubic, leaveTangent=0.2132, time=0.697, value=0.1503}
          ]}
         GreenCurve {keys [
           {arriveTangent=-0.4331, interpMode=RCIM_Cubic, leaveTangent=-0.4331, tangentMode=RCTM_User, time=0.1195, value=0.2261}
           {interpMode=RCIM_Cubic, leaveTangent=2.0123, time=0.697, value=0.1146}
          ]}
         RedCurve {keys [
           {arriveTangent=-4.577, interpMode=RCIM_Cubic, leaveTangent=-4.577, tangentMode=RCTM_User, time=0.1195, value=2}
           {interpMode=RCIM_Cubic, leaveTangent=-2.7231, time=0.697, value=2}
          ]}
        }}))
     "Float (Alpha)"=&Multiply_Float(
      A=&FloatFromCurve(FloatCurve=DataInterface() {propertyValues {Curve {keys [
           {arriveTangent=-0.9, interpMode=RCIM_Cubic, time=0, value=0}
           <at 0.05 1 RCIM_Cubic>
           {interpMode=RCIM_Cubic, leaveTangent=-1.5912, time=1, value=0}
          ]}}})
      B=&FNTimeOfDayScalar(Night Scale=0.7, Morning Scale=0.5, Evening Scale=0.6)
     )
    )) {isSetParameters=true, module=SetVariables_#6}
   &FN_Light_Attributes(
    Color=&ColorFromCurve(ColorCurve=DataInterface() {propertyValues {
       AlphaCurve {keys []}
       BlueCurve {keys [<at 0 0 RCIM_Cubic>,<at 0.0858 0>,<at 0.1804 0>,<at 0.9 0.0729>,<at 1 0 RCIM_Cubic>]}
       GreenCurve {keys [<at 0 0 RCIM_Cubic>,<at 0.0858 0>,<at 0.1804 0>,<at 0.9 0.107>,<at 1 0 RCIM_Cubic>]}
       RedCurve {keys [<at 0 0 RCIM_Cubic>,<at 0.0858 30>,<at 0.1804 10>,<at 0.9 2>,<at 1 0 RCIM_Cubic>]}
      }})
    Create Color Attribute=true
    Create Exponent Attribute=true
    Create Position Attribute=true
    Create Radius Attribute=true
    Exponent=3
    Position=$Particles.SkeletalMeshLocation.SampledPosition
    Radius=&FloatFromCurve(
     FloatCurve=DataInterface() {propertyValues {Curve {keys [<at 0 0.4>,<at 0.1 1.5>,<at 0.2 1>,<at 0.5 1>,<at 0.8 0.8>,<at 1 0.4>]}}}
     Scale Curve=&Multiply_Float(A=&MakeCustomFloatFromBool(Bool=$User.bIsFrontEndPreview, True Float=40, False Float=80), B=1)
    )
    Scale Color=&FloatFromCurve(
     FloatCurve=DataInterface() {propertyValues {Curve {keys [<at 0 0.5>,<at 0.1 1.1>,<at 0.1401 0.4933>,<at 0.2 1.3742>,<at 0.2354 0.5667>,<at 0.3 1.1>,<at 0.3404 0.4667>,<at 0.4098 1.3991>,<at 0.4342 0.48>,<at 0.5 1.1>,<at 0.5311 0.5133>,<at 0.5829 1.3306>,<at 0.6376 0.5133>,<at 0.7 1.1>,<at 0.741 0.5067>,<at 0.7989 1.4926>,<at 0.8459 0.3133>,<at 0.9 1.1>,<at 1 0>]}}}
     Scale Curve=&Multiply_Float(A=5, B=&FNTimeOfDayScalar(Night Scale=0.3, Morning Scale=0.5, Day Scale=0.8, Evening Scale=0.4))
    )
   )
   &SolveForcesAndVelocity(Clamp Velocity=false, Limit Acceleration=false)
  ]
  rendererClasses [
   &NiagaraSpriteRendererProperties
   &NiagaraMeshRendererProperties
   &NiagaraLightRendererProperties
  ]
  renderers [
   {rendererClass=&NiagaraSpriteRendererProperties}
   {rendererClass=&NiagaraSpriteRendererProperties, rendererIndex=1}
   {rendererClass=&NiagaraSpriteRendererProperties, rendererIndex=2}
   {rendererClass=&NiagaraMeshRendererProperties, rendererIndex=3}
   {rendererClass=&NiagaraLightRendererProperties, rendererIndex=4}
  ]
 }
 Eyeglow001 {
  bEnabled=false
  emitterUpdateScript [
   &EmitterState(Inactive Response=::NewEnumerator1, Life Cycle Mode=::NewEnumerator1, Loop Behavior=::NewEnumerator1, Loop Duration=2, UseLoopDelay=false)
   &SpawnBurst_Instantaneous(Use Spawn Probability=false, UseLoopCountLimit=false)
  ]
  particleSpawnScript [
   &InitializeParticle(
    Color=<rgba 0.15 0.15 0.15 1.5>
    Lifetime=0.5
    Mesh Scale=<0.08 0.08 0.08>
    Mesh Scale Mode=::NewEnumerator4
    Sprite Size=<160 160>
    Sprite Size Mode=::NewEnumerator4
    UsePositionOffset=false
    Write Material Random=true
   )
   &InitialMeshOrientation(Mesh Orientation Mode=::NewEnumerator2, Orientation Vector=<0 0 1>, Rotation=<0 0.25 0>, Use Rotation Vector=true)
   &CameraOffset(Camera Offset Amount=5)
   &AddVelocity(Velocity=<0 0 0.001>)
  ]
  particleUpdateScript [
   &ParticleState()
   &ScaleMeshSize(Scale Factor=&Multiply_Vector(
     A=&ScaleFloatByCurve(
      InitialValue:Vector3f=<1 1 1>
      ScaleCurve:NiagaraDataInterfaceVectorCurve=DataInterface() {propertyValues {
        XCurve {keys [<at 0 0.25>,<at 0.1 1>,<at 0.9 1>,<at 1 0>]}
        YCurve {keys [<at 0 1.25>,<at 0.1 2.25>,<at 0.9 2.25>,<at 1 0>]}
        ZCurve {keys [<at 0 1.25>,<at 0.1 2.25>,<at 0.9 2.25>,<at 1 0>]}
       }}
      CurveIndex:NiagaraFloat=$Particles.NormalizedAge
     )
     B=&Divide_Vector(A:Vector3f=<1 1 1>, B:Vector3f=$Engine.Owner.Scale)
    ))
   &ScaleSpriteSize(Uniform Curve Sprite Scale=DataInterface() {propertyValues {Curve {keys [
        <at 0 0>
        <at 0.1 1>
        <at 0.1512 0.7174>
        <at 0.2132 1>
        <at 0.2641 0.7539>
        <at 0.299 1>
        <at 0.3563 0.7037>
        <at 0.4151 1>
        <at 0.4564 0.8017>
        <at 0.5041 1>
        <at 0.5502 0.7561>
        <at 0.5836 1>
        <at 0.6551 0.7653>
        <at 0.6981 1>
        <at 0.7521 0.7721>
        <at 0.7871 1>
        <at 0.8475 0.7106>
        <at 0.9 1>
        {tangentMode=RCTM_User, time=1, value=0}
       ]}}})
   /SetVariables(Particles.Size002:Vector2f=&Multiply_Vector2D(A=$Particles.SpriteSize, B=<0.6 0.9>)) {isSetParameters=true, module=SetVariables_#2}
   /SetVariables(Particles.Size003:Vector2f=&Multiply_Vector2D(A=<150 50>, B=<1 3>)) {isSetParameters=true, module=SetVariables_#3}
   &ScaleColor(
    Scale Alpha=&Multiply_Float(
     A=&FloatFromCurve(
      FloatCurve=DataInterface() {propertyValues {Curve {keys [
          {arriveTangent=4.4475, time=0, value=0}
          {arriveTangent=5.3518, leaveTangent=-3.1774, time=0.1, value=1.2}
          {arriveTangent=-1.4972, leaveTangent=-1.4972, time=0.1523, value=0.7}
          {arriveTangent=-0.1667, leaveTangent=-0.1667, time=0.2052, value=1.1711}
          <at 0.258 0.7713>
          <at 0.2926 1.1775>
          {arriveTangent=-0.25, leaveTangent=-0.25, time=0.3608, value=0.689}
          <at 0.3865 1.193>
          <at 0.4263 0.8642>
          <at 0.4931 1.1191>
          {arriveTangent=-0.25, leaveTangent=-0.25, time=0.5428, value=0.7548}
          <at 0.5846 1.1063>
          <at 0.6571 0.8076>
          <at 0.6976 1.1685>
          {arriveTangent=0.1667, leaveTangent=0.1667, time=0.7542, value=0.7713}
          <at 0.7906 1.1138>
          <at 0.8535 0.8807>
          {arriveTangent=-3.3333, leaveTangent=-3.3333, time=0.9, value=1.2}
          <at 1 0>
         ]}}}
      Scale Curve=&FNTimeOfDayScalar(Night Scale=0.6, Morning Scale=0.7, Day Scale=0.8, Evening Scale=0.6)
     )
     B=&FN_FrontEndFloatModifier(Gameplay=1, Front End=0.3)
    )
    Scale RGB=&MakeVectorFromLinearColorRGB(Color=&ColorFromCurve(ColorCurve=DataInterface() {propertyValues {
        AlphaCurve {keys []}
        BlueCurve {keys [<at 0.7887 0 RCIM_Cubic>,<at 0.9801 0.0012 RCIM_Cubic>]}
        GreenCurve {keys [<at 0.7887 0.0063 RCIM_Cubic>,<at 0.9801 0 RCIM_Cubic>]}
        RedCurve {keys [<at 0.7887 0.4375 RCIM_Cubic>,<at 0.9801 0.1562 RCIM_Cubic>]}
       }}))
    ScaleA=true
    ScaleRGB=true
   )
   /SetVariables(Particles.color002:LinearColor=&MakeLinearColorFromVectorAndFloat(
     "Vector (RGB)"=&MakeVectorFromLinearColorRGB(Color=<rgba 3 0.12 0.18 1>)
     "Float (Alpha)"=&FloatFromCurve(
      FloatCurve=DataInterface() {propertyValues {Curve {keys [<at 0 1>,<at 0.1 2>,<at 0.2005 1.7767>,<at 0.306 1.9927>,<at 0.3889 1.7548>,<at 0.5026 2.0076>,<at 0.5773 1.7477>,<at 0.6601 1.9791>,<at 0.7283 1.7134>,<at 0.8209 1.9848>,<at 1 0>]}}}
      Scale Curve=&FN_PickFloatDependOnPlatform(TargetPlatformActive=5, TargetPlatformNoActive=1)
     )
    )) {isSetParameters=true, module=SetVariables_#4}
   /SetVariables(Particles.Color003:LinearColor=&MakeLinearColorFromVectorAndFloat(
     "Vector (RGB)"=&MakeVectorFromLinearColorRGB(Color=<rgba 1 0.04 0.05 1>)
     "Float (Alpha)"=&Multiply_Float(
      A=&FloatFromCurve(
       FloatCurve=DataInterface() {propertyValues {Curve {keys [<at 0 0.8>,<at 0.1 1>,<at 0.2 0.6>,<at 0.3 1>,<at 0.4 0.7>,<at 0.5 0.95>,<at 0.6 0.55>,<at 0.7 1>,<at 0.8 0.55>,<at 0.9 0.8>,<at 1 0>]}}}
       Scale Curve=&FNTimeOfDayScalar(Night Scale=0.2, Morning Scale=0.1, Day Scale=0.4, Evening Scale=0.2)
      )
      B=&Multiply_Float(
       A=&FN_FrontEndFloatModifier(Gameplay=0.5, Front End=0.15)
       B=&FN_PickFloatDependOnPlatform(TargetPlatformActive=1, TargetPlatformNoActive=0.4)
      )
     )
    )) {isSetParameters=true, module=SetVariables_#5}
   /SetVariables(Particles.MeshColor:LinearColor=&MakeLinearColorFromVectorAndFloat(
     "Vector (RGB)"=&MakeVectorFromLinearColorRGB(Color=&ColorFromCurve(ColorCurve=DataInterface() {propertyValues {
         AlphaCurve {keys []}
         BlueCurve {keys [
           {arriveTangent=-1.8762, interpMode=RCIM_Cubic, leaveTangent=-1.8762, tangentMode=RCTM_User, time=0.1298, value=0.3256}
           {interpMode=RCIM_Cubic, leaveTangent=0.2132, time=0.3441, value=0.1503}
          ]}
         GreenCurve {keys [
           {arriveTangent=-0.4331, interpMode=RCIM_Cubic, leaveTangent=-0.4331, tangentMode=RCTM_User, time=0.1298, value=0.2261}
           {interpMode=RCIM_Cubic, leaveTangent=2.0123, time=0.3441, value=0.1146}
          ]}
         RedCurve {keys [
           {arriveTangent=-4.577, interpMode=RCIM_Cubic, leaveTangent=-4.577, tangentMode=RCTM_User, time=0.1298, value=2}
           {interpMode=RCIM_Cubic, leaveTangent=-2.7231, time=0.3441, value=2}
          ]}
        }}))
     "Float (Alpha)"=&Multiply_Float(
      A=&FloatFromCurve(
       FloatCurve=DataInterface() {propertyValues {Curve {keys [
           {arriveTangent=-0.9, time=0, value=0}
           {arriveTangent=-0.0998, leaveTangent=-0.0998, time=0.05, value=1}
           {arriveTangent=-1.0872, leaveTangent=-1.0872, time=0.1804, value=0.7734}
           <at 0.2985 0.937>
           <at 0.4531 0.6834>
           <at 0.5813 0.9005>
           <at 0.7511 0.6392>
           <at 0.8564 0.8>
           {arriveTangent=-1.5912, leaveTangent=-1.5912, time=1, value=0}
          ]}}}
       Scale Curve=&FN_FrontEndFloatModifier(Gameplay=1, Front End=0.5)
      )
      B=&FNTimeOfDayScalar(Night Scale=0.7, Morning Scale=0.5, Evening Scale=0.6)
     )
    )) {isSetParameters=true, module=SetVariables_#6}
   &FN_Light_Attributes(
    Color=&ColorFromCurve(ColorCurve=DataInterface() {propertyValues {
       AlphaCurve {keys []}
       BlueCurve {keys [<at 0 0 RCIM_Cubic>,<at 0.0858 0>,<at 0.1804 0>,<at 0.9 0.0729>,<at 1 0 RCIM_Cubic>]}
       GreenCurve {keys [<at 0 0 RCIM_Cubic>,<at 0.0858 0>,<at 0.1804 0>,<at 0.9 0.107>,<at 1 0 RCIM_Cubic>]}
       RedCurve {keys [<at 0 0 RCIM_Cubic>,<at 0.0858 30>,<at 0.1804 10>,<at 0.9 2>,<at 1 0 RCIM_Cubic>]}
      }})
    Create Color Attribute=true
    Create Exponent Attribute=true
    Create Position Attribute=true
    Create Radius Attribute=true
    Exponent=3
    Position=$Particles.SkeletalMeshLocation.SampledPosition
    Radius=&FloatFromCurve(
     FloatCurve=DataInterface() {propertyValues {Curve {keys [<at 0 0.4>,<at 0.1 1.5>,<at 0.2 1>,<at 0.5 1>,<at 0.8 0.8>,<at 1 0.4>]}}}
     Scale Curve=&Multiply_Float(A=&MakeCustomFloatFromBool(Bool=$User.bIsFrontEndPreview, True Float=40, False Float=80), B=1)
    )
    Scale Color=&FloatFromCurve(
     FloatCurve=DataInterface() {propertyValues {Curve {keys [<at 0 0.5>,<at 0.1 1.1>,<at 0.1401 0.4933>,<at 0.2 1.3742>,<at 0.2354 0.5667>,<at 0.3 1.1>,<at 0.3404 0.4667>,<at 0.4098 1.3991>,<at 0.4342 0.48>,<at 0.5 1.1>,<at 0.5311 0.5133>,<at 0.5829 1.3306>,<at 0.6376 0.5133>,<at 0.7 1.1>,<at 0.741 0.5067>,<at 0.7989 1.4926>,<at 0.8459 0.3133>,<at 0.9 1.1>,<at 1 0>]}}}
     Scale Curve=&Multiply_Float(A=5, B=&FNTimeOfDayScalar(Night Scale=0.3, Morning Scale=0.5, Day Scale=0.8, Evening Scale=0.4))
    )
   )
   &SolveForcesAndVelocity(Clamp Velocity=false, Limit Acceleration=false)
  ]
  rendererClasses [
   &NiagaraSpriteRendererProperties
   &NiagaraMeshRendererProperties
   &NiagaraLightRendererProperties
  ]
  renderers [
   {rendererClass=&NiagaraSpriteRendererProperties}
   {rendererClass=&NiagaraSpriteRendererProperties, rendererIndex=1}
   {rendererClass=&NiagaraSpriteRendererProperties, rendererIndex=2}
   {rendererClass=&NiagaraMeshRendererProperties, rendererIndex=3}
   {rendererClass=&NiagaraLightRendererProperties, rendererIndex=4}
  ]
 }
 Feathers {
  emitterUpdateScript [
   &EmitterState(
    Inactive Response=::NewEnumerator1
    Life Cycle Mode=::NewEnumerator1
    Loop Behavior=::NewEnumerator1
    Loop Delay=0.4
    Loop Duration=2
    UseLoopDelay=false
   )
   &SpawnBurst_Instantaneous(Spawn Count=3, Use Spawn Probability=false, UseLoopCountLimit=false)
  ]
  particleSpawnScript [
   &InitializeParticle(
    Color=&MergeHelper_ColorSwitch(
     0=<rgba 0.0118 0.0108 0.0273 1>
     1=<rgba 0.1198 0.0038 0.0037 1>
     2=?
     3=?
     Index=&RandomRangeInt(Minimum=0, Maximum=1)
     Num=::NewEnumerator0
    )
    Lifetime Max=1.54
    Lifetime Mode=::NewEnumerator1
    Sprite Rotation Mode=::NewEnumerator2
    Sprite Size=<10 18>
    Sprite Size Mode=::NewEnumerator4
    UsePositionOffset=true
    Write Material Random=true
   )
   &CameraOffset(Camera Offset Amount=30)
   &AddVelocity(
    Module.Use Falloff Distance=false
    Origin Offset=$StackContext.ShapeLocation.ShapePosition
    Override Seed=false
    Velocity Mode=::NewEnumerator1
    Velocity Speed=100
   ) {enabled=false}
   &AddVelocity(
    Rotation Coordinate Space=::World
    Velocity=&RandomRangeVector(Minimum=<-15 -15 70>, Maximum=<15 15 90>)
    Velocity Speed Scale=2
   ) {module=AddVelocity001}
   &ShapeLocation(MaskSpawnGroup=false, Override Seed=false, Sphere Radius=40, UseOwnerScale=true)
   &SubUVAnimation(
    Evaluation Type=::NewEnumerator0
    "Sprite Renderer "=DataInterface() {propertyValues {SpriteRenderer=&self:Feathers_0.NiagaraSpriteRendererProperties_0}}
    SubUV Animation Mode:ENiagara_SubUVLookupModeV2=::NewEnumerator2
    UseEndFrame=false
    UseStartFrame=false
   ) {module=SubUVAnimation001}
  ]
  particleUpdateScript [
   &ParticleState()
   &ScaleVelocity(Velocity Scale=&VectorFromCurve(VectorCurve=DataInterface() {propertyValues {
       XCurve {keys [{arriveTangent=-1.756, interpMode=RCIM_Cubic, leaveTangent=-1.756, tangentMode=RCTM_User, time=0, value=1}, <at 1 0>]}
       YCurve {keys [{arriveTangent=-1.756, interpMode=RCIM_Cubic, leaveTangent=-1.756, tangentMode=RCTM_User, time=0, value=1}, <at 1 0>]}
       ZCurve {keys [{arriveTangent=-1.756, interpMode=RCIM_Cubic, leaveTangent=-1.756, tangentMode=RCTM_User, time=0, value=1}, <at 1 0>]}
      }}))
   &SpriteRotationRate(Rotation Rate=&RandomRangeFloat(Minimum=-250, Maximum=250))
   &ScaleSpriteSize(
    Scale Sprite Size Mode=::NewEnumerator0
    Uniform Scale Factor=&FloatFromCurve(
     FloatCurve=DataInterface() {propertyValues {Curve {
        defaultValue=1
        keys [
         {arriveTangent=4.1785, interpMode=RCIM_Cubic, leaveTangent=4.1785, tangentMode=RCTM_User, time=0, value=1}
         {arriveTangent=1.1179, interpMode=RCIM_Cubic, leaveTangent=1.1179, tangentMode=RCTM_User, time=0.2042, value=1.6}
         {arriveTangent=0.0224, interpMode=RCIM_Cubic, leaveTangent=0.0224, tangentMode=RCTM_User, time=1, value=2}
        ]
       }}}
     Scale Curve=0.2
    )
   )
   &ScaleColor(
    Scale Alpha=&Multiply_Float(
     A=&RandomRangeFloat(Minimum=0.5, Maximum=1)
     B=&FloatFromCurve(
      FloatCurve=DataInterface() {propertyValues {Curve {keys [
          {arriveTangent=3.7485, interpMode=RCIM_Cubic, leaveTangent=3.7485, tangentMode=RCTM_User, time=0, value=0.35}
          {interpMode=RCIM_Cubic, tangentMode=RCTM_User, time=0.5, value=1}
          {tangentMode=RCTM_User, time=1, value=0}
         ]}}}
      Scale Curve=2
     )
    )
    ScaleA=true
    ScaleRGB=true
   )
   &Drag(Drag=6, Use Linear Drag=true, Use Rotational Drag=true) {enabled=false}
   &GravityForce()
   &CurlNoiseForce(
    Bias Noise=false
    Noise Frequency=55
    Noise Strength=&RandomRangeFloat(Minimum=400, Maximum=800)
    Pan Noise=false
   )
   &SolveForcesAndVelocity(Clamp Velocity=false, Limit Acceleration=false)
  ]
  rendererClasses [&NiagaraSpriteRendererProperties]
  renderers [{rendererClass=&NiagaraSpriteRendererProperties}]
 }
 Flare {
  emitterUpdateScript [
   &EmitterState(Life Cycle Mode=::NewEnumerator1, Loop Behavior=::NewEnumerator1, Loop Delay=0.1, Loop Duration=1, UseLoopDelay=false)
   &SpawnBurst_Instantaneous(Spawn Count=1, Use Spawn Probability=false)
  ]
  particleSpawnScript [
   &InitializeParticle(
    Color=<rgba 5 0.0664 0 0.3>
    Lifetime=0.33
    Position Offset=<0 0 5>
    Position Offset Coordinate Space=::World
    Sprite Rotation Mode=::NewEnumerator2
    Sprite Size Mode=::NewEnumerator3
    Uniform Sprite Size=200
    UsePositionOffset=true
    Write Material Random=true
   )
   &DynamicMaterialParameters(
    Index 0 Param 1=&RandomRangeFloat(Minimum=0.4, Maximum=0.6)
    Index 0 Param 2=1
    Index 0 Param 3=1
    Index 0 Param 4=1
    Module.Param0WriteEnabled=true
    Module.Param1WriteEnabled=true
    Module.Param2WriteEnabled=true
    Module.Param3WriteEnabled=true
   )
   /SetVariables(Particles.Dark_Glow_CameraOffset=&MakeCustomFloatFromBool(Bool=$User.bIsFrontEndPreview, True Float=1, False Float=50)) {isSetParameters=true}
   &CameraOffset(Camera Offset Amount=&MakeCustomFloatFromBool(Bool=$User.bIsFrontEndPreview, True Float=1, False Float=30))
  ]
  particleUpdateScript [
   /SetVariables(Particles.Dark_Glow=&MakeLinearColorFromVectorAndFloat(
     "Vector (RGB)"=&MakeVectorFromLinearColorRGB(Color=&MultiplyLinearColor(
       Linear Color=<rgba 0.2031 0.0176 0.0486 1>
       Scale Factor=&MakeVector4_FromFloat(Float=&FNTimeOfDayScalar(Night Scale=0.5, Morning Scale=0.5, Evening Scale=0.3))
      ))
     "Float (Alpha)"=&Multiply_Float(
      A=&Multiply_Float(A=&FNTimeOfDayScalar(Night Scale=0.5, Morning Scale=0.5, Evening Scale=1), B=1.2)
      B=&FloatFromCurve(FloatCurve=DataInterface() {propertyValues {Curve {keys [<at 0 0>,<at 0.2 1>,<at 0.7 0.6>,<at 1 0>]}}}, Scale Curve=0.7)
     )
    )) {isSetParameters=true}
   /SetVariables(Particles.Dark_Glow_Size=&Multiply_Vector2(
     A=&Vector2DFromFloat(Value=220)
     B=&Vector2DFromFloat(Value=&FloatFromCurve(FloatCurve=DataInterface() {propertyValues {Curve {keys [<at 0 1.5 RCIM_Cubic>,<at 1 0.5>]}}}, Scale Curve=0.7))
    )) {isSetParameters=true}
   &ParticleState()
   &ScaleColor(
    Scale Alpha=&Multiply_Float(
     A=&FNTimeOfDayScalar(Night Scale=0.5, Morning Scale=0.5, Evening Scale=0.5)
     B=&FloatFromCurve(
      FloatCurve=DataInterface() {propertyValues {Curve {keys [<at 0 0>,<at 0.2 1>,<at 0.8 1>,<at 1 0>]}}}
      Scale Curve=&FNTimeOfDayScalar(Night Scale=0.5, Morning Scale=0.5, Evening Scale=0.5)
     )
    )
    Scale RGB=<2 2 2>
    ScaleA=true
    ScaleRGB=true
   )
   &ScaleSpriteSize(Scale Factor=&Vector2DFromFloat(Value=&FloatFromCurve(FloatCurve=DataInterface() {propertyValues {Curve {keys [<at 0 1.5 RCIM_Cubic>,<at 1 0>]}}}, Scale Curve=0.7)))
  ]
  rendererClasses [&NiagaraSpriteRendererProperties]
  renderers [
   {rendererClass=&NiagaraSpriteRendererProperties}
   {rendererClass=&NiagaraSpriteRendererProperties, rendererIndex=1}
  ]
 }
 Wave {
  bEnabled=false
  emitterUpdateScript [
   &EmitterState(Life Cycle Mode=::NewEnumerator1, Loop Behavior=::NewEnumerator1, Loop Duration=1, UseLoopDelay=true)
   &SpawnBurst_Instantaneous(Spawn Count=1, Use Spawn Probability=false)
  ]
  particleSpawnScript [
   &InitializeParticle(
    Color=<rgba 1.5 0.1131 0.0844 1>
    Lifetime=0.23
    Mesh Scale Mode=::NewEnumerator3
    Mesh Uniform Scale=0.6
    Position Offset=<0 0 5>
    Position Offset Coordinate Space=::World
    UsePositionOffset=true
    Write Material Random=true
   )
   &InitialMeshOrientation(Orientation Axis=<0 0 1>, Orientation Vector=<0 0 1>, Rotation=<0 0.25 0>, Use Orientation Vector=true, Use Rotation Vector=true)
   &InitialMeshOrientation(
    Orientation Axis=<0 0 0>
    Orientation Vector=<0 0 0>
    Rotation=&RandomRangeVector(Minimum=<0 0 0>, Maximum=<1 0 0>)
    Use Orientation Vector=false
    Use Rotation Vector=true
   ) {module=InitialMeshOrientation001}
   &CameraOffset(Camera Offset Amount=30)
  ]
  particleUpdateScript [
   &ParticleState()
   &ScaleColor(
    Linear Color Curve=DataInterface() {propertyValues {
      AlphaCurve {keys [<at 0 1>,<at 0.3576 1>,<at 1 0>]}
      BlueCurve {keys [<at 0 0.1667>,<at 0.1684 0>,<at 0.3654 0.1412>,<at 1 0.0825>]}
      GreenCurve {keys [<at 0 0>,<at 0.1684 0>,<at 0.3654 0>,<at 1 0>]}
      RedCurve {keys [<at 0 1>,<at 0.1684 1>,<at 0.3654 1>,<at 1 1>]}
     }}
    Scale Mode=::NewEnumerator2
   )
   &DynamicMaterialParameters(
    Index 0 Param 1=&FloatFromCurve(FloatCurve=DataInterface() {propertyValues {Curve {keys [{arriveTangent=-2.1015, interpMode=RCIM_Cubic, leaveTangent=-2.1015, tangentMode=RCTM_User, time=0, value=0.5}, <at 1 -0.15>]}}})
    Index 0 Param 2=&FloatFromCurve(FloatCurve=DataInterface() {propertyValues {Curve {keys [{arriveTangent=-0.3464, interpMode=RCIM_Cubic, leaveTangent=-0.3464, tangentMode=RCTM_User, time=0, value=1}, <at 1 0.9>]}}})
    Index 0 Param 3=&FloatFromCurve(FloatCurve=DataInterface() {propertyValues {Curve {keys [<at 0 0>,<at 0.15 1>,<at 1 1.2>]}}})
    Module.Param0WriteEnabled=true
    Module.Param1WriteEnabled=true
    Module.Param2WriteEnabled=true
    Module.Param3WriteEnabled=false
   )
   &ScaleColor(
    Scale Alpha=&FNTimeOfDayScalar(Night Scale=0.2, Morning Scale=0.3, Evening Scale=0.3)
    Scale RGB=&VectorFromFloat(Value=&FloatFromCurve(FloatCurve=DataInterface() {propertyValues {Curve {keys [<at 0 10>,<at 0.15 3>,<at 1 1>]}}}))
    ScaleA=true
    ScaleRGB=true
   ) {module=ScaleColor001}
   &ScaleMeshSize(Scale Factor=&VectorFromFloat(Value=&FloatFromCurve(
      FloatCurve=DataInterface() {propertyValues {Curve {keys [{arriveTangent=2.322, interpMode=RCIM_Cubic, leaveTangent=2.322, tangentMode=RCTM_User, time=0, value=0}, <at 1 2>]}}}
      Scale Curve=0.7
     )))
   &ScaleColor(
    Scale RGB=&Multiply_VectorByFloat(Vector=<1 1 1>, Float=&FNTimeOfDayScalar(Night Scale=1, Morning Scale=1, Evening Scale=1))
    ScaleA=true
    ScaleRGB=true
   ) {module=ScaleColor002}
  ]
  rendererClasses [&NiagaraMeshRendererProperties]
  renderers [{rendererClass=&NiagaraMeshRendererProperties, rendererIndex=0}]
 }
}
summary {
 systemName=NS_Emote_MistRaven_Book_01
 emitters [
  {
   emitterName=Em_Smoke_Cloud
   rendererClasses [&NiagaraSpriteRendererProperties, &NiagaraLightRendererProperties]
  }
  {emitterName=Em_Smoke_Cloud_Detail, rendererClasses [&NiagaraSpriteRendererProperties]}
  {emitterName=Feathers, rendererClasses [&NiagaraSpriteRendererProperties]}
  {emitterName=Flare, rendererClasses [&NiagaraSpriteRendererProperties]}
  {bEnabled=false, emitterName=Wave, rendererClasses [&NiagaraMeshRendererProperties]}
  {
   emitterName=Eyeglow
   rendererClasses [
    &NiagaraSpriteRendererProperties
    &NiagaraMeshRendererProperties
    &NiagaraLightRendererProperties
   ]
  }
  {
   bEnabled=false
   emitterName=Eyeglow001
   rendererClasses [
    &NiagaraSpriteRendererProperties
    &NiagaraMeshRendererProperties
    &NiagaraLightRendererProperties
   ]
  }
 ]
 userVariables [
  User.Ascended {type=NiagaraBool, defaultValue=true}
  User.bIsFrontEnd {type=NiagaraBool}
  User.bIsFrontEndPreview {type=NiagaraBool}
  User.SpeedLineCount {type=NiagaraInt32}
 ]
}
