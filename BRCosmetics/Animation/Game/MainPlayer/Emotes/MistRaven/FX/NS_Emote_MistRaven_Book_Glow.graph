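# Niagara effect for the book glow in the MistRaven player emote, pairing a hot orange-red flare with dark smoke.
# Flare bursts one large sprite that fades and shrinks, with alpha scaled by time of day and front-end context, and also computes separate Dark_Glow color and size attributes.
# SmokeDetailed spawns short-lived, rotating flipbook puffs in dark blue or red from a small sphere, growing and fading over their lifetime.
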NiagaraSystem /BRCosmetics/Animation/Game/MainPlayer/Emotes/MistRaven/FX/NS_Emote_MistRaven_Book_Glow
dependencies {
 usedDataInterfaces [
  &NiagaraDataInterfaceVectorCurve
  &NiagaraDataInterfaceCurve
  &NiagaraDataInterfaceVector2DCurve
  &NiagaraDataInterfaceSpriteRendererInfo
 ]
 usedDynamicInputs [&Vector2DFromCurve, &RandomRangeVector2D]
 usedModules [&SystemState]
}
emitters {
 Flare {
  emitterUpdateScript [
   &EmitterState(Life Cycle Mode=::NewEnumerator1, Loop Behavior=::NewEnumerator1, Loop Delay=0.1, Loop Duration=1, UseLoopDelay=false)
   &SpawnBurst_Instantaneous(Spawn Count=1, Use Spawn Probability=false)
  ]
  particleSpawnScript [
   &InitializeParticle(
    Color=<rgba 6 0.0797 0 0.3>
    Lifetime=2.4
    Position Offset=<0 0 1.2>
    Position Offset Coordinate Space=::World
    Sprite Size=<200 200>
    Sprite Size Mode=::NewEnumerator4
    UsePositionOffset=true
    Write Material Random=true
   )
   &DynamicMaterialParameters(
    Index 0 Param 1=&RandomRangeFloat(Minimum=0.4, Maximum=0.6)
    Index 0 Param 2=0.1
    Index 0 Param 3=1
    Index 0 Param 4=1
    Module.Index0Float0Write=true
    Module.Index0Float1Write=true
    Module.Index0Float2Write=true
    Module.Index0Float3Write=true
   )
   /SetVariables(Particles.Dark_Glow_CameraOffset=&MakeCustomFloatFromBool(Bool=$User.bIsFrontEndPreview, True Float=1, False Float=1)) {isSetParameters=true}
   &CameraOffset(Camera Offset Amount=-4)
  ]
  particleUpdateScript [
   /SetVariables(Particles.Dark_Glow=&MakeLinearColorFromVectorAndFloat(
     "Vector (RGB)"=&MakeVectorFromLinearColorRGB(Color=&MultiplyLinearColor(
       Linear Color=<rgba 0.2031 0.0176 0.0486 1>
       Scale Factor=&MakeVector4_FromFloat(Float=&FNTimeOfDayScalar(Night Scale=0.5, Morning Scale=0.5, Evening Scale=0.3))
      ))
     "Float (Alpha)"=&Multiply_Float(
      A=&Multiply_Float(A=&FNTimeOfDayScalar(Night Scale=0.5, Morning Scale=0.5, Evening Scale=1), B=1.2)
      B=&FloatFromCurve(
       FloatCurve=DataInterface() {propertyValues {Curve {keys [
           <at 0 0>
           {arriveTangent=1.1268, interpMode=RCIM_Cubic, leaveTangent=1.1268, time=0.3528, value=1}
           {arriveTangent=-1.5451, interpMode=RCIM_Cubic, leaveTangent=-1.5451, time=0.7, value=0.7888}
           <at 1 0>
          ]}}}
       Scale Curve=0.4
      )
     )
    )) {isSetParameters=true}
   /SetVariables(Particles.Dark_Glow_Size=&Multiply_Vector2(
     A=<220 220>
     B=&Vector2DFromFloat(Value=&FloatFromCurve(FloatCurve=DataInterface() {propertyValues {Curve {keys [<at 0 1.5 RCIM_Cubic>,<at 1 0.8>]}}}, Scale Curve=0.45))
    )) {isSetParameters=true}
   &ParticleState()
   &ScaleColor(
    Scale Alpha=&Multiply_Float(
     A=&FNTimeOfDayScalar(Night Scale=0.5, Morning Scale=0.5, Evening Scale=0.5)
     B=&FloatFromCurve(
      FloatCurve=DataInterface() {propertyValues {Curve {keys [
          <at 0 0>
          {arriveTangent=0.3864, interpMode=RCIM_Cubic, leaveTangent=0.3864, tangentMode=RCTM_User, time=0.4326, value=1}
          {arriveTangent=-0.424, interpMode=RCIM_Cubic, leaveTangent=-0.424, tangentMode=RCTM_User, time=0.5125, value=1}
          <at 1 0>
         ]}}}
      Scale Curve=&Multiply_Float(
       A=&FN_FrontEndFloatModifier(Gameplay=1, Front End=0.5)
       B=&FNTimeOfDayScalar(Night Scale=0.5, Morning Scale=0.5, Evening Scale=0.5)
      )
     )
    )
    ScaleA=true
    ScaleRGB=true
   )
   &ScaleSpriteSize(Scale Factor=&Vector2DFromFloat(Value=&FloatFromCurve(FloatCurve=DataInterface() {propertyValues {Curve {keys [<at 0 1.5 RCIM_Cubic>,<at 1 0.8>]}}}, Scale Curve=0.45)))
  ]
  rendererClasses [&NiagaraSpriteRendererProperties]
  renderers [{rendererClass=&NiagaraSpriteRendererProperties}]
 }
 SmokeDetailed {
  emitterUpdateScript [
   &EmitterState(Life Cycle Mode=::NewEnumerator1, Loop Behavior=::NewEnumerator1, Loop Delay=0.4, Loop Duration=1, UseLoopDelay=false)
   &SpawnRate(SpawnRate=6, Use Spawn Probability=false)
  ]
  particleSpawnScript [
   &InitializeParticle(
    Color=&MergeHelper_ColorSwitch(
     0=<rgba 0.0118 0.0108 0.0273 1>
     1=<rgba 0.1198 0.0038 0.0037 1>
     2=?
     3=?
     Index=&RandomRangeInt(Minimum=0, Maximum=1)
     Num=::NewEnumerator0
    )
    Lifetime Max=1
    Lifetime Min=0.8
    Lifetime Mode=::NewEnumerator1
    Sprite Rotation Mode=::NewEnumerator2
    Sprite Size Max=<100 100>
    Sprite Size Min=<80 80>
    Sprite Size Mode=::NewEnumerator2
    UsePositionOffset=true
    Write Material Random=true
   )
   &CameraOffset(Camera Offset Amount=30)
   &ShapeLocation(MaskSpawnGroup=false, Override Seed=false, Sphere Radius=10, UseOwnerScale=true)
  ]
  particleUpdateScript [
   &ParticleState()
   &ScaleVelocity(Velocity Scale=&VectorFromCurve(VectorCurve=DataInterface() {propertyValues {
       XCurve {keys [{arriveTangent=-1.756, interpMode=RCIM_Cubic, leaveTangent=-1.756, tangentMode=RCTM_User, time=0, value=1}, <at 1 0>]}
       YCurve {keys [{arriveTangent=-1.756, interpMode=RCIM_Cubic, leaveTangent=-1.756, tangentMode=RCTM_User, time=0, value=1}, <at 1 0>]}
       ZCurve {keys [{arriveTangent=-1.756, interpMode=RCIM_Cubic, leaveTangent=-1.756, tangentMode=RCTM_User, time=0, value=1}, <at 1 0>]}
      }}))
   &SpriteRotationRate(Rotation Rate=&RandomRangeFloat(Minimum=-50, Maximum=50))
   &ScaleSpriteSize(
    Scale Sprite Size Mode=::NewEnumerator0
    Uniform Scale Factor=&FloatFromCurve(
     FloatCurve=DataInterface() {propertyValues {Curve {
        defaultValue=1
        keys [
         {arriveTangent=4.1785, interpMode=RCIM_Cubic, leaveTangent=4.1785, tangentMode=RCTM_User, time=0, value=1}
         {arriveTangent=1.1179, interpMode=RCIM_Cubic, leaveTangent=1.1179, tangentMode=RCTM_User, time=0.2042, value=1.6}
         {arriveTangent=0.0224, interpMode=RCIM_Cubic, leaveTangent=0.0224, tangentMode=RCTM_User, time=1, value=2}
        ]
       }}}
     Scale Curve=0.2
    )
   )
   &SubUVAnimation(
    End Frame Range Override=30
    "Sprite Renderer "=DataInterface() {propertyValues {SpriteRenderer=&self:SmokeDetailed_0.NiagaraSpriteRendererProperties_0}}
    Start Frame Range Override=&RandomRangeInt(Minimum=0, Maximum=5)
    UseEndFrame=true
    UseStartFrame=true
   )
   &ScaleColor(
    Scale Alpha=&Multiply_Float(
     A=&RandomRangeFloat(Minimum=0.5, Maximum=1)
     B=&FloatFromCurve(
      FloatCurve=DataInterface() {propertyValues {Curve {keys [
          {arriveTangent=3.7485, interpMode=RCIM_Cubic, leaveTangent=3.7485, tangentMode=RCTM_User, time=0, value=0.35}
          {interpMode=RCIM_Cubic, tangentMode=RCTM_User, time=0.5, value=1}
          {tangentMode=RCTM_User, time=1, value=0}
         ]}}}
      Scale Curve=1.5
     )
    )
    ScaleA=true
    ScaleRGB=true
   )
   &SolveForcesAndVelocity(Clamp Velocity=false, Limit Acceleration=false)
  ]
  rendererClasses [&NiagaraSpriteRendererProperties]
  renderers [{rendererClass=&NiagaraSpriteRendererProperties}]
 }
}
summary {
 systemName=NS_Emote_MistRaven_Book_Glow
 emitters [
  {emitterName=SmokeDetailed, rendererClasses [&NiagaraSpriteRendererProperties]}
  {emitterName=Flare, rendererClasses [&NiagaraSpriteRendererProperties]}
 ]
 userVariables [
  User.bIsFrontEnd {type=NiagaraBool}
  User.bIsFrontEndPreview {type=NiagaraBool}
 ]
}
