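# Niagara system for the red glow of the book in the MistRaven player emote, lasting about four seconds.
# Burst-spawned Flare sprites and MeshGlow meshes form the core, with a flickering red light.
# Two dark-and-red smoke emitters, one SubUV-animated, drift downward while Swirl meshes with time-driven material parameters ramp in and fade with system age.
# Glow brightness adapts to time of day and front-end versus gameplay, with platform-dependent Flare alpha.
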NiagaraSystem /BRCosmetics/Animation/Game/MainPlayer/Emotes/MistRaven/FX/NS_Emote_MistRaven_Book_GlowMain
dependencies {
 usedDataInterfaces [
  &NiagaraDataInterfaceCurve
  &NiagaraDataInterfaceVector2DCurve
  &NiagaraDataInterfaceSpriteRendererInfo
  &NiagaraDataInterfacePlatformSet
  &NiagaraDataInterfaceVectorCurve
  &NiagaraDataInterfaceColorCurve
  &NiagaraDataInterfaceArrayColor
 ]
 usedDynamicInputs [&Vector2DFromCurve, &RandomRangeVector2D]
 usedModules [&SystemState]
}
emitters {
 Flare {
  emitterUpdateScript [
   &EmitterState(Life Cycle Mode=::NewEnumerator1, Loop Behavior=::NewEnumerator1, Loop Delay=0.1, Loop Duration=1, UseLoopDelay=false)
   &SpawnBurst_Instantaneous(Spawn Count=1, Use Spawn Probability=false)
  ]
  particleSpawnScript [
   &InitializeParticle(
    Color=<rgba 4 0.0532 0 0.3>
    Lifetime=4
    Position Offset=<0 0 1>
    Position Offset Coordinate Space=::World
    Sprite Rotation Mode=::NewEnumerator2
    Sprite Size Mode=::NewEnumerator3
    Uniform Sprite Size=&FN_FrontEndFloatModifier(Gameplay=200, Front End=170)
    UsePositionOffset=true
    Write Material Random=true
   )
   &DynamicMaterialParameters(
    Index 0 Param 1=&RandomRangeFloat(Minimum=0.4, Maximum=0.6)
    Index 0 Param 2=0.3
    Index 0 Param 3=1
    Index 0 Param 4=1
    Module.Index0Float0Write=true
    Module.Index0Float1Write=true
    Module.Index0Float2Write=true
    Module.Index0Float3Write=true
   )
   /SetVariables(Particles.Dark_Glow_CameraOffset=1) {isSetParameters=true}
   &CameraOffset(Camera Offset Amount=5)
  ]
  particleUpdateScript [
   /SetVariables(Particles.Dark_Glow=&MakeLinearColorFromVectorAndFloat(
     "Vector (RGB)"=&MakeVectorFromLinearColorRGB(Color=&MultiplyLinearColor(
       Linear Color=<rgba 0.2031 0.0176 0.0486 1>
       Scale Factor=&MakeVector4_FromFloat(Float=&FNTimeOfDayScalar(Night Scale=0.5, Morning Scale=0.5, Evening Scale=0.3))
      ))
     "Float (Alpha)"=&Multiply_Float(
      A=&Multiply_Float(A=&FNTimeOfDayScalar(Night Scale=0.5, Morning Scale=0.5, Evening Scale=1), B=1.2)
      B=&FloatFromCurve(
       FloatCurve=DataInterface() {propertyValues {Curve {keys [
           <at 0 0>
           {arriveTangent=0.8216, interpMode=RCIM_Cubic, leaveTangent=0.8216, time=0.3528, value=1}
           {arriveTangent=-1.5451, interpMode=RCIM_Cubic, leaveTangent=-1.5451, time=0.96, value=0.7888}
           <at 1 0>
          ]}}}
       Scale Curve=&FN_FrontEndFloatModifier(
        Gameplay=&FN_PickFloatDependOnPlatform(TargetPlatform=DataInterface() {propertyValues {Platforms {qualityLevelMask=-29}}}, TargetPlatformActive=0.3, TargetPlatformNoActive=0.7)
        Front End=0.15
       )
      )
     )
    )) {isSetParameters=true}
   /SetVariables(Particles.Dark_Glow_Size=&Multiply_Vector2(
     A=&Vector2DFromFloat(Value=&FN_FrontEndFloatModifier(Gameplay=200, Front End=200))
     B=&Vector2DFromFloat(Value=&FloatFromCurve(FloatCurve=DataInterface() {propertyValues {Curve {keys [<at 0 1.5 RCIM_Cubic>,<at 1 2>]}}}, Scale Curve=0.5))
    )) {isSetParameters=true}
   &ParticleState()
   &ScaleColor(
    Scale Alpha=&Multiply_Float(
     A=&FN_FrontEndFloatModifier(
      Gameplay=&FN_PickFloatDependOnPlatform(TargetPlatform=DataInterface() {propertyValues {Platforms {qualityLevelMask=-29}}}, TargetPlatformActive=0.45, TargetPlatformNoActive=1)
      Front End=0.4
     )
     B=&Multiply_Float(
      A=&FNTimeOfDayScalar(Night Scale=0.5, Morning Scale=0.5, Evening Scale=0.5)
      B=&FloatFromCurve(
       FloatCurve=DataInterface() {propertyValues {Curve {keys [
           <at 0.093 0 RCIM_Cubic>
           {arriveTangent=0.3864, interpMode=RCIM_Cubic, leaveTangent=0.3864, tangentMode=RCTM_User, time=0.4326, value=1}
           {arriveTangent=-0.424, interpMode=RCIM_Cubic, leaveTangent=-0.424, tangentMode=RCTM_User, time=0.96, value=1}
           <at 1 0>
          ]}}}
       Scale Curve=&FNTimeOfDayScalar(Night Scale=0.5, Morning Scale=0.5, Evening Scale=0.5)
      )
     )
    )
    Scale RGB=&Multiply_Vector(
     A=<1 1 1>
     B=&VectorFromFloat(Value=&FN_FrontEndFloatModifier(Gameplay=1, Front End=1))
    )
    ScaleA=true
    ScaleRGB=true
   )
   &ScaleSpriteSize(Uniform Curve Scale=0.5, Uniform Curve Sprite Scale=DataInterface() {propertyValues {Curve {keys [<at 0 1.5 RCIM_Cubic>,<at 1 2>]}}})
  ]
  rendererClasses [&NiagaraSpriteRendererProperties]
  renderers [
   {rendererClass=&NiagaraSpriteRendererProperties}
   {rendererClass=&NiagaraSpriteRendererProperties, rendererIndex=1}
  ]
 }
 MeshGlow {
  emitterUpdateScript [
   &EmitterState(Inactive Response=::NewEnumerator1, Life Cycle Mode=::NewEnumerator1, Loop Behavior=::NewEnumerator1, UseLoopDelay=false)
   &SpawnBurst_Instantaneous(Spawn Count=2, Use Spawn Probability=false, UseLoopCountLimit=false)
   /SetVariables(Emitter.PanSpeed=0.5) {isSetParameters=true}
  ]
  particleSpawnScript [
   &FN_SetSpawnGroup()
   &InitializeParticle(
    Lifetime=&MergeHelper_FloatSwitch(0=4, 1=4.2, 2=?, 3=?, Index=&SpawnGroup(), Num=::NewEnumerator0)
    Mesh Scale=&MergeHelper_VectorSwitch(0=<0.08 0.08 0.08>, 1=<0.2 0.196 1>, 2=?, 3=?, Index=&SpawnGroup(), Num=::NewEnumerator0)
    Mesh Scale Mode=::NewEnumerator4
    Position Offset=<0 0 1>
    Position Offset Coordinate Space=::World
    Sprite Size=<160 160>
    Sprite Size Mode=::NewEnumerator4
    UsePositionOffset=true
    Write Material Random=true
   )
   &InitialMeshOrientation(
    Mesh Axis To Orient=&MergeHelper_VectorSwitch(0=<1 0 0>, 1=<0 0 1>, 2=?, 3=?, Index=&SpawnGroup(), Num=::NewEnumerator0)
    Mesh Orientation Mode=::NewEnumerator2
    Orientation Vector=&MergeHelper_VectorSwitch(0=<0 0 1>, 1=<0.024 0 1>, 2=?, 3=?, Index=&SpawnGroup(), Num=::NewEnumerator0)
    Rotation=&MergeHelper_VectorSwitch(0=<0 0.25 0>, 1=<0 0 0>, 2=?, 3=?, Index=&SpawnGroup(), Num=::NewEnumerator0)
    Use Rotation Vector=true
   )
   &CameraOffset(Camera Offset Amount=&MergeHelper_FloatSwitch(0=5, 1=1, 2=?, 3=?, Index=&SpawnGroup(), Num=::NewEnumerator0))
   &AddVelocity(Velocity=<0 0 0.001>)
   /SetVariables(Particles.VisibilityTag=&SpawnGroup()) {isSetParameters=true}
   /SetVariables() {isSetParameters=true}
  ]
  particleUpdateScript [
   &ParticleState()
   &ScaleMeshSize(Scale Factor=&MergeHelper_VectorSwitch(
     0=&Multiply_Vector(
      A=&ScaleFloatByCurve(
       InitialValue:Vector3f=<1 1 1>
       ScaleCurve:NiagaraDataInterfaceVectorCurve=DataInterface() {propertyValues {
         XCurve {keys [<at 0 0.25>,<at 0.1 1>,<at 0.9 1>,<at 1 0>]}
         YCurve {keys [<at 0 1.25>,<at 0.1 2.25>,<at 0.9 2.25>,<at 1 0>]}
         ZCurve {keys [<at 0 1.25>,<at 0.1 2.25>,<at 0.9 2.25>,<at 1 0>]}
        }}
       CurveIndex:NiagaraFloat=$Particles.NormalizedAge
      )
      B=&Divide_Vector(A:Vector3f=<1.17 1.17 1.17>, B:Vector3f=$Engine.Owner.Scale)
     )
     1=<1 1 1>
     2=?
     3=?
     Index=&SpawnGroup()
     Num=::NewEnumerator0
    ))
   &ScaleSpriteSize(Uniform Curve Sprite Scale=DataInterface() {propertyValues {Curve {keys [
        <at 0 0>
        <at 0.1 1>
        <at 0.1512 0.7174>
        <at 0.2132 1>
        <at 0.2641 0.7539>
        <at 0.299 1>
        <at 0.3563 0.7037>
        <at 0.4151 1>
        <at 0.4564 0.8017>
        <at 0.5041 1>
        <at 0.5502 0.7561>
        <at 0.5836 1>
        <at 0.6551 0.7653>
        <at 0.6981 1>
        <at 0.7521 0.7721>
        <at 0.7871 1>
        <at 0.8475 0.7106>
        <at 0.9 1>
        {tangentMode=RCTM_User, time=1, value=0}
       ]}}})
   &ScaleColor(
    Scale Alpha=&FloatFromCurve(
     FloatCurve=DataInterface() {propertyValues {Curve {keys [
         <at 0 0 RCIM_Cubic>
         {arriveTangent=0.4615, interpMode=RCIM_Cubic, leaveTangent=0.4615, tangentMode=RCTM_User, time=0.2353, value=1}
         {arriveTangent=-0.1753, interpMode=RCIM_Cubic, leaveTangent=-0.1753, tangentMode=RCTM_User, time=0.8231, value=1}
         {arriveTangent=-1.8362, leaveTangent=-1.8362, time=1, value=0}
        ]}}}
     Scale Curve=&FN_FrontEndFloatModifier(Gameplay=0.6, Front End=0.4)
    )
    Scale RGB=&MakeVectorFromLinearColorRGB(Color=&Multiply_LinearColorByFloat(
      Linear Color=<rgba 2 0.1875 0.2216 1>
      Float=&FNTimeOfDayScalar(Night Scale=0.6, Morning Scale=0.8, Evening Scale=0.7)
     ))
    ScaleA=true
    ScaleRGB=true
   )
   /SetVariables(Particles.MeshColor:LinearColor=&MakeLinearColorFromVectorAndFloat(
     "Vector (RGB)"=&MakeVectorFromLinearColorRGB(Color=&ColorFromCurve(ColorCurve=DataInterface() {propertyValues {
         AlphaCurve {keys []}
         BlueCurve {keys [
           {arriveTangent=-1.8762, interpMode=RCIM_Cubic, leaveTangent=-1.8762, tangentMode=RCTM_User, time=0.1195, value=0.179}
           {interpMode=RCIM_Cubic, leaveTangent=0.2132, time=0.697, value=0.1}
          ]}
         GreenCurve {keys [
           {arriveTangent=-0.4331, interpMode=RCIM_Cubic, leaveTangent=-0.4331, tangentMode=RCTM_User, time=0.1195, value=0.2363}
           {interpMode=RCIM_Cubic, leaveTangent=2.0123, time=0.697, value=0.1}
          ]}
         RedCurve {keys [
           {arriveTangent=-4.577, interpMode=RCIM_Cubic, leaveTangent=-4.577, tangentMode=RCTM_User, time=0.1195, value=2}
           {interpMode=RCIM_Cubic, leaveTangent=-2.7231, time=0.697, value=2}
          ]}
        }}))
     "Float (Alpha)"=&Multiply_Float(
      A=&FloatFromCurve(
       FloatCurve=DataInterface() {propertyValues {Curve {keys [
           {arriveTangent=-0.9, interpMode=RCIM_Cubic, time=0.1044, value=0}
           {arriveTangent=0.8768, interpMode=RCIM_Cubic, leaveTangent=0.8768, time=0.1794, value=1}
           {arriveTangent=-1.2187, interpMode=RCIM_Cubic, leaveTangent=-1.2187, time=0.96, value=0.7502}
           {interpMode=RCIM_Cubic, leaveTangent=-0.0037, time=1, value=0}
          ]}}}
       Scale Curve=&FN_FrontEndFloatModifier(Gameplay=0.3, Front End=0.2)
      )
      B=&FNTimeOfDayScalar(Night Scale=0.7, Morning Scale=0.5, Evening Scale=0.6)
     )
    )) {isSetParameters=true}
   &FN_Light_Attributes(
    Color=&ColorFromCurve(ColorCurve=DataInterface() {propertyValues {
       AlphaCurve {keys []}
       BlueCurve {keys [<at 0 0 RCIM_Cubic>,<at 0.0858 0>,<at 0.1804 0>]}
       GreenCurve {keys [<at 0 0 RCIM_Cubic>,<at 0.0858 0>,<at 0.1804 0>]}
       RedCurve {keys [<at 0 0 RCIM_Cubic>,<at 0.0858 30>,<at 0.1804 10>]}
      }})
    Create Color Attribute=true
    Create Exponent Attribute=true
    Create Position Attribute=true
    Create Radius Attribute=true
    Exponent=1
    Radius=&FloatFromCurve(FloatCurve=DataInterface() {propertyValues {Curve {keys [<at 0 0 RCIM_Cubic>,<at 0.4 1>]}}}, Scale Curve=130)
    Scale Color=&FloatFromCurve(
     FloatCurve=DataInterface() {propertyValues {Curve {keys [
         <at 0 0 RCIM_Cubic>
         {arriveTangent=2.2281, interpMode=RCIM_Cubic, leaveTangent=2.2281, time=0.3, value=1.123}
         {arriveTangent=2.7615, interpMode=RCIM_Cubic, leaveTangent=2.7615, time=0.3361, value=0.7488}
         <at 0.4 1.3991>
         <at 0.4342 0.48>
         <at 0.5 1.1>
         <at 0.5311 0.5133>
         <at 0.5829 1.3306>
         <at 0.6376 0.5133>
         <at 0.7 1.1>
         <at 0.741 0.5067>
         <at 0.7989 1.4926>
         <at 0.8459 0.3133>
         <at 0.9 1.1>
         <at 1 0>
        ]}}}
     Scale Curve=&Multiply_Float(
      A=1.7
      B=&FNTimeOfDayScalar(Night Scale=0.3, Morning Scale=0.5, Day Scale=0.8, Evening Scale=0.4)
     )
    )
   )
   &SolveForcesAndVelocity(Clamp Velocity=false, Limit Acceleration=false)
   /SetVariables(Particles.MeshlGlowsDyn:Vector4f=&MakeVector4(
     X=0.2
     Y=1
     Z=&Lerp_Float(
      A=0
      B=0.01
      Alpha=&FloatFromCurve(FloatCurve=DataInterface() {propertyValues {Curve {keys [
           <at 0 0 RCIM_Cubic>
           {arriveTangent=0.4909, interpMode=RCIM_Cubic, leaveTangent=0.4909, tangentMode=RCTM_User, time=0.3, value=1}
           <at 0.9 1>
          ]}}})
     )
     W=40
    )) {isSetParameters=true}
  ]
  rendererClasses [&NiagaraMeshRendererProperties, &NiagaraLightRendererProperties]
  renderers [
   {rendererClass=&NiagaraMeshRendererProperties, rendererIndex=0}
   {rendererClass=&NiagaraLightRendererProperties, rendererIndex=1}
   {rendererClass=&NiagaraMeshRendererProperties, rendererIndex=2}
  ]
 }
 Smoke {
  emitterUpdateScript [
   &EmitterState(Life Cycle Mode=::NewEnumerator1, Loop Behavior=::NewEnumerator1, Loop Delay=1.9, Loop Duration=3.5, UseLoopDelay=false)
   &SpawnRate(SpawnRate=3, Use Spawn Probability=false)
  ]
  particleSpawnScript [
   &InitializeParticle(
    Color=&MergeHelper_ColorSwitch(
     0=<rgba 0.0118 0.0108 0.0273 1>
     1=<rgba 0.401 0.0126 0.0125 2>
     2=?
     3=?
     Index=&RandomRangeInt(Minimum=0, Maximum=1)
     Num=::NewEnumerator0
    )
    Lifetime Max=1.15
    Lifetime Min=0.9
    Lifetime Mode=::NewEnumerator1
    Sprite Rotation Mode=::NewEnumerator2
    Sprite Size Mode=::NewEnumerator1
    Uniform Sprite Size Max=100
    Uniform Sprite Size Min=75
    UsePositionOffset=false
    Write Material Random=true
   )
   &ShapeLocation(MaskSpawnGroup=false, Override Seed=false, Sphere Radius=6, UseOwnerScale=true)
   &AddVelocity(Rotation Coordinate Space=::World, Velocity=&RandomRangeVector(Minimum=<0 0 0>, Maximum=<0 0 -30>))
  ]
  particleUpdateScript [
   &ParticleState()
   &SpriteRotationRate(Rotation Rate=&RandomRangeFloat(Minimum=-70, Maximum=70))
   &ScaleSpriteSize(
    Uniform Curve Scale=0.2
    Uniform Curve Sprite Scale=DataInterface() {propertyValues {Curve {
       defaultValue=1
       keys [
        {arriveTangent=4.1785, interpMode=RCIM_Cubic, leaveTangent=4.1785, tangentMode=RCTM_User, time=0, value=1}
        {arriveTangent=1.1179, interpMode=RCIM_Cubic, leaveTangent=1.1179, tangentMode=RCTM_User, time=0.1, value=1.875}
        {arriveTangent=0.0224, interpMode=RCIM_Cubic, leaveTangent=0.0224, tangentMode=RCTM_User, time=1, value=2}
       ]
      }}}
   )
   &SolveForcesAndVelocity(Clamp Velocity=false, Limit Acceleration=false)
   &ScaleColor(
    Scale Alpha=&FloatFromCurve(FloatCurve=DataInterface() {propertyValues {Curve {keys [
         {interpMode=RCIM_Cubic, tangentMode=RCTM_User, time=0, value=0.4}
         {arriveTangent=-2.0189, interpMode=RCIM_Cubic, leaveTangent=-2.1015, tangentMode=RCTM_Break, time=0.2, value=1.35}
         {arriveTangent=-1.1521, interpMode=RCIM_Cubic, tangentMode=RCTM_Break, time=1, value=0}
        ]}}})
    ScaleA=true
    ScaleRGB=true
   )
  ]
  rendererClasses [&NiagaraSpriteRendererProperties]
  renderers [{rendererClass=&NiagaraSpriteRendererProperties}]
 }
 SmokeDetailed {
  emitterUpdateScript [
   &EmitterState(Life Cycle Mode=::NewEnumerator1, Loop Behavior=::NewEnumerator1, Loop Delay=0.4, Loop Duration=3.5, UseLoopDelay=false)
   &SpawnRate(SpawnRate=4, Use Spawn Probability=false)
  ]
  particleSpawnScript [
   &InitializeParticle(
    Color=&MergeHelper_ColorSwitch(
     0=<rgba 0.0118 0.0108 0.0273 1>
     1=<rgba 0.1198 0.0038 0.0037 1>
     2=?
     3=?
     Index=&RandomRangeInt(Minimum=0, Maximum=1)
     Num=::NewEnumerator0
    )
    Lifetime Max=1
    Lifetime Min=0.8
    Lifetime Mode=::NewEnumerator1
    Sprite Rotation Mode=::NewEnumerator2
    Sprite Size Max=<100 100>
    Sprite Size Min=<80 80>
    Sprite Size Mode=::NewEnumerator2
    UsePositionOffset=true
    Write Material Random=true
   )
   &CameraOffset(Camera Offset Amount=30)
   &ShapeLocation(MaskSpawnGroup=false, Override Seed=false, Sphere Radius=10, UseOwnerScale=true)
   &AddVelocity(Rotation Coordinate Space=::World, Velocity=&RandomRangeVector(Minimum=<0 0 0>, Maximum=<0 0 -30>))
  ]
  particleUpdateScript [
   &ParticleState()
   &SpriteRotationRate(Rotation Rate=&RandomRangeFloat(Minimum=-50, Maximum=50))
   &ScaleSpriteSize(
    Scale Sprite Size Mode=::NewEnumerator0
    Uniform Scale Factor=&FloatFromCurve(
     FloatCurve=DataInterface() {propertyValues {Curve {
        defaultValue=1
        keys [
         {arriveTangent=4.1785, interpMode=RCIM_Cubic, leaveTangent=4.1785, tangentMode=RCTM_User, time=0, value=1}
         {arriveTangent=1.1179, interpMode=RCIM_Cubic, leaveTangent=1.1179, tangentMode=RCTM_User, time=0.2042, value=1.6}
         {arriveTangent=0.0224, interpMode=RCIM_Cubic, leaveTangent=0.0224, tangentMode=RCTM_User, time=1, value=2}
        ]
       }}}
     Scale Curve=0.25
    )
   )
   &SubUVAnimation(
    End Frame Range Override=30
    "Sprite Renderer "=DataInterface() {propertyValues {SpriteRenderer=&self:SmokeDetailed_1.NiagaraSpriteRendererProperties_0}}
    Start Frame Range Override=&RandomRangeInt(Minimum=0, Maximum=5)
    UseEndFrame=true
    UseStartFrame=true
   )
   &ScaleColor(
    Scale Alpha=&Multiply_Float(
     A=&RandomRangeFloat(Minimum=0.5, Maximum=1)
     B=&FloatFromCurve(
      FloatCurve=DataInterface() {propertyValues {Curve {keys [
          {arriveTangent=3.7485, interpMode=RCIM_Cubic, leaveTangent=3.7485, tangentMode=RCTM_User, time=0, value=0.35}
          {interpMode=RCIM_Cubic, tangentMode=RCTM_User, time=0.5, value=1}
          {tangentMode=RCTM_User, time=1, value=0}
         ]}}}
      Scale Curve=1.4
     )
    )
    ScaleA=true
    ScaleRGB=true
   )
   &SolveForcesAndVelocity(Clamp Velocity=false, Limit Acceleration=false)
  ]
  rendererClasses [&NiagaraSpriteRendererProperties]
  renderers [{rendererClass=&NiagaraSpriteRendererProperties}]
 }
 Swirl {
  emitterUpdateScript [
   &EmitterState(Inactive Response=::NewEnumerator1, Life Cycle Mode=::NewEnumerator1, Loop Behavior=::NewEnumerator1, UseLoopDelay=false)
   &SpawnRate(
    SpawnRate=&FloatFromCurve(FloatCurve=DataInterface() {propertyValues {Curve {keys [<at 0.75 0>,<at 1 1>]}}}, CurveIndex=$System.Age, Scale Curve=5)
    Use Spawn Probability=false
   )
  ]
  particleSpawnScript [
   &FN_SetSpawnGroup(Number of Spawn Groups=2)
   &InitializeParticle(
    Color=<rgba 1 1 1 2>
    Lifetime=&RandomRangeFloat(Minimum=1.3, Maximum=1.6)
    Mesh Scale=&RandomRangeVector(Minimum=<0.16 0.16 0.15>, Maximum=<0.2 0.16 0.3>)
    Mesh Scale Mode=::NewEnumerator4
    Position Offset=&MergeHelper_VectorSwitch(0=<-5.8574 -4.0085 14.7845>, 1=<-5.8574 -4.0085 14.7845>, 2=?, 3=?, Index=&SpawnGroup(), Num=::NewEnumerator0)
    UsePositionOffset=true
    Write Material Random=true
   )
   &InitialMeshOrientation(
    Mesh Axis To Orient=<0 0 1>
    Mesh Orientation Mode=::NewEnumerator2
    Orientation Vector=<0 1 0>
    Rotation=&MergeHelper_VectorSwitch(
     0=&RandomRangeVector(Minimum=<0 0.072 0.136>, Maximum=<0.05 0.1 0.136>)
     1=&RandomRangeVector(Minimum=<0.584 0.184 0>, Maximum=<0.7 0.21 0>)
     2=?
     3=?
     Index=&SpawnGroup()
     Num=::NewEnumerator0
    )
    Use Rotation Vector=true
   )
   /SetVariables(Particles.RandomTimeFloat:NiagaraFloat=&RandomRangeFloat(Minimum=-1.2, Maximum=-1.4)) {isSetParameters=true}
  ]
  particleUpdateScript [
   &ParticleState()
   &ScaleColor(
    Scale Alpha=&FloatFromCurve(
     FloatCurve=DataInterface() {propertyValues {Curve {keys [<at 0.7822 1 RCIM_Cubic>,<at 1 0>]}}}
     Scale Curve=&Multiply_Float(
      A=0.22
      B=&FloatFromCurve(FloatCurve=DataInterface() {propertyValues {Curve {keys [<at 3.5 1 RCIM_Cubic>,<at 4 0>]}}}, CurveIndex=$System.Age)
     )
    )
    Scale RGB=&MakeVectorFromLinearColorRGB(Color=&MultiplyLinearColor(
      Linear Color=&SelectLinearColorFromArray(
       Array Sampling Mode=::NewEnumerator0
       Direct Array Index=&RandomRangeInt(Minimum=0, Maximum=1)
       Color Selection Array=DataInterface() {propertyValues {ColorData [<rgba 0.5469 0.0142 0.021 1>, <rgba 0.0469 0.0012 0.0018 1.3>]}}
       Random Seed=0
      )
      Scale Factor=<1 1 1 1>
     ))
    ScaleA=true
    ScaleRGB=true
   )
   &DynamicMaterialParameters(
    Index 0 Param 1=&Multiply_Float(A=$Engine.RealTime, B=$Particles.RandomTimeFloat)
    Index 0 Param 2=&Multiply_Float(A=$Engine.RealTime, B=$Particles.RandomTimeFloat)
    Index 0 Param 3=69.6374
    Index 1 Param 1=&FloatFromCurve(FloatCurve=DataInterface() {propertyValues {Curve {keys [<at 0 0 RCIM_Cubic>,<at 1 1>]}}})
    Module.Index0Param0Write=true
    Module.Index0Param1Write=true
    Module.Index0Param2Write=true
    Module.Index0Param3Write=true
    Module.Index1Param0Write=true
    Module.Index1Param1Write=true
    Module.Index1Param2Write=true
    Module.Index1Param3Write=true
    Write Parameter Index 1=true
   )
  ]
  rendererClasses [&NiagaraMeshRendererProperties]
  renderers [
   {rendererClass=&NiagaraMeshRendererProperties, rendererIndex=0}
   {rendererClass=&NiagaraMeshRendererProperties, rendererIndex=1}
  ]
 }
}
summary {
 systemName=NS_Emote_MistRaven_Book_GlowMain
 emitters [
  {emitterName=Smoke, rendererClasses [&NiagaraSpriteRendererProperties]}
  {emitterName=SmokeDetailed, rendererClasses [&NiagaraSpriteRendererProperties]}
  {emitterName=Flare, rendererClasses [&NiagaraSpriteRendererProperties]}
  {emitterName=MeshGlow, rendererClasses [&NiagaraMeshRendererProperties, &NiagaraLightRendererProperties]}
  {emitterName=Swirl, rendererClasses [&NiagaraMeshRendererProperties]}
 ]
 userVariables [
  User.bIsFrontEnd {type=NiagaraBool}
  User.bIsFrontEndPreview {type=NiagaraBool}
 ]
}
