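# Niagara system for the idle glow on the book in the MistRaven emote, built from one CapsuleGlow sprite emitter.
# Burst-spawned, long-lived red sprites take position and alignment from their spawn group, forming a center glow and two mirrored sides.
# A waveform flickers alpha, dimmed in the front end.
# A separate Dark_Glow color parameter shares that flicker and scales with time of day, alongside two sprite renderers.
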NiagaraSystem /BRCosmetics/Animation/Game/MainPlayer/Emotes/MistRaven/FX/NS_Emote_MistRaven_Book_IdleGlow
dependencies {
 usedDataInterfaces [&NiagaraDataInterfaceCurve, &NiagaraDataInterfaceVector2DCurve]
 usedDynamicInputs []
 usedModules [&SystemState]
}
emitters {CapsuleGlow {
  emitterUpdateScript [
   &EmitterState(Inactive Response=::NewEnumerator1, Life Cycle Mode=::NewEnumerator1, Loop Behavior=::NewEnumerator1, Loop Duration=11, UseLoopDelay=false)
   &SpawnBurst_Instantaneous(Use Spawn Probability=false, UseLoopCountLimit=false)
  ]
  particleSpawnScript [
   &InitializeParticle(
    Color=<rgba 2 0.0266 0 1>
    Lifetime=11
    Position=&MergeHelper_VectorSwitch(0=<0 0 2>, 1=<9.6565 0 6.7364>, 2=<-9.6565 0 6.7364>, 3=?, Index=&SpawnGroup(), Num=::NewEnumerator1)
    Position Mode=::NewEnumerator1
    Sprite Rotation Mode=::NewEnumerator0
    Sprite Size=<77 103>
    Sprite Size Mode=::NewEnumerator4
    UsePositionOffset=false
    Write Material Random=true
   )
   &CameraOffset(Camera Offset Amount=-6)
  ]
  particleUpdateScript [
   &ParticleState()
   /SetVariables(Particles.Waveform:NiagaraFloat=&Waveform("Waveform [1]"=::NewEnumerator12, "[1] Frequency"=7, "[1] Amplitude Min/Max"=<0.7 1>)) {isSetParameters=true}
   /SetVariables(Particles.Dark_Glow=&MakeLinearColorFromVectorAndFloat(
     "Vector (RGB)"=&MakeVectorFromLinearColorRGB(Color=&MultiplyLinearColor(
       Linear Color=<rgba 0.2031 0.0176 0.0486 1>
       Scale Factor=&MakeVector4_FromFloat(Float=&FNTimeOfDayScalar(Night Scale=0.5, Morning Scale=0.5, Evening Scale=0.3))
      ))
     "Float (Alpha)"=&Multiply_Float(
      A=&Multiply_Float(A=&FNTimeOfDayScalar(Night Scale=0.5, Morning Scale=0.5, Evening Scale=1), B=1.2)
      B=&FloatFromCurve(
       FloatCurve=DataInterface() {propertyValues {Curve {keys [
           <at 0 0 RCIM_Cubic>
           {arriveTangent=-0.1383, interpMode=RCIM_Cubic, leaveTangent=-0.1383, tangentMode=RCTM_User, time=0.0664, value=1}
           <at 1 1 RCIM_Cubic>
          ]}}}
       Scale Curve=&Multiply_Float(A=0.17, B=$Particles.Waveform)
      )
     )
    )) {isSetParameters=true}
   &ScaleSpriteSize(Scale Sprite Size Mode=::NewEnumerator0)
   &SpriteFacingAndAlignment(
    Sprite Alignment=&MergeHelper_VectorSwitch(0=<0 1 0>, 1=<0 0 1>, 2=<0 0 1>, 3=?, Index=&SpawnGroup(), Num=::NewEnumerator1)
    Sprite Facing=<0.7068 0.07 0.3>
    WriteAlignment=true
    WriteFacing=false
   )
   &ScaleColor(
    Scale Alpha=&FloatFromCurve(
     FloatCurve=DataInterface() {propertyValues {Curve {keys [
         <at 0 0 RCIM_Cubic>
         {arriveTangent=-0.1383, interpMode=RCIM_Cubic, leaveTangent=-0.1383, tangentMode=RCTM_User, time=0.0664, value=1}
         <at 1 1 RCIM_Cubic>
        ]}}}
     Scale Curve=&Multiply_Float(A=&FN_FrontEndFloatModifier(Gameplay=0.05, Front End=0.025), B=$Particles.Waveform)
    )
    Scale RGB=&VectorFromFloat(Value=&FloatFromCurve(FloatCurve=DataInterface() {propertyValues {Curve {keys [<at 0 1 RCIM_Cubic>,<at 1 1>]}}}))
    ScaleA=true
    ScaleRGB=true
   )
  ]
  rendererClasses [&NiagaraSpriteRendererProperties]
  renderers [
   {rendererClass=&NiagaraSpriteRendererProperties}
   {rendererClass=&NiagaraSpriteRendererProperties, rendererIndex=1}
  ]
 }}
summary {
 systemName=NS_Emote_MistRaven_Book_IdleGlow
 emitters [{emitterName=CapsuleGlow, rendererClasses [&NiagaraSpriteRendererProperties]}]
 userVariables [
  User.bIsFrontEnd {type=NiagaraBool}
  User.bIsFrontEndPreview {type=NiagaraBool}
 ]
}
