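# Niagara one-shot burst effect for the MistRaven emote, played, per its name, as the book closes.
# A red Flash, drawn by several sprite renderers plus a light, pairs with three short CapsuleGlow streaks at fixed positions.
# Dark and deep-red feathers and flipbook smoke puff outward as curl-noise sparks scatter, stretched by speed.
# Flash brightness adapts to time of day, front-end versus gameplay, and platform, while sparks dim by time of day.
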NiagaraSystem /BRCosmetics/Animation/Game/MainPlayer/Emotes/MistRaven/FX/NS_Emote_MistRaven_Book_MainClose
dependencies {
 usedDataInterfaces [
  &NiagaraDataInterfaceSpriteRendererInfo
  &NiagaraDataInterfaceVectorCurve
  &NiagaraDataInterfaceCurve
  &NiagaraDataInterfaceVector2DCurve
  &NiagaraDataInterfaceColorCurve
  &NiagaraDataInterfacePlatformSet
 ]
 usedDynamicInputs [&Vector2DFromCurve, &RandomRangeVector2D]
 usedModules [&SystemState]
}
emitters {
 CapsuleGlow {
  emitterUpdateScript [
   &EmitterState(Inactive Response=::NewEnumerator1, Life Cycle Mode=::NewEnumerator1, Loop Behavior=::NewEnumerator1, Loop Duration=1, UseLoopDelay=false)
   &SpawnBurst_Instantaneous(Spawn Count=3, Spawn Time=0.07, Use Spawn Probability=false, UseLoopCountLimit=false)
  ]
  particleSpawnScript [
   &FN_SetSpawnGroup(Number of Spawn Groups=3)
   &InitializeParticle(
    Color=<rgba 9 0.093 0 1>
    Lifetime=0.25
    Position=&MergeHelper_VectorSwitch(0=<0 0 12.8>, 1=<9.6565 0 6.7364>, 2=<-9.6565 0 6.7364>, 3=?, Index=&SpawnGroup(), Num=::NewEnumerator1)
    Position Mode=::NewEnumerator1
    Sprite Rotation Mode=::NewEnumerator0
    Sprite Size=&MergeHelper_Vector2Switch(0=<7 31.8854>, 1=<7 21.349>, 2=<7 21.349>, 3=?, 4=?, Index=&SpawnGroup(), Num=::NewEnumerator1)
    Sprite Size Mode=::NewEnumerator4
    UsePositionOffset=false
    Write Material Random=true
   )
   &CameraOffset(Camera Offset Amount=4)
  ]
  particleUpdateScript [
   &ParticleState()
   &ScaleSpriteSize(Scale Sprite Size Mode=::NewEnumerator0)
   &SpriteFacingAndAlignment(
    Sprite Alignment=&MergeHelper_VectorSwitch(0=<1 0 0>, 1=<0 0 1>, 2=<0 0 1>, 3=?, Index=&SpawnGroup(), Num=::NewEnumerator1)
    Sprite Facing=<0.7068 0.07 0.3>
    WriteAlignment=true
    WriteFacing=false
   )
   &ScaleColor(
    Scale Alpha=&FloatFromCurve(FloatCurve=DataInterface() {propertyValues {Curve {keys [<at 0 0 RCIM_Cubic>,<at 0.05 1 RCIM_Cubic>,<at 1 0 RCIM_Cubic>]}}})
    Scale RGB=&VectorFromFloat(Value=&FloatFromCurve(FloatCurve=DataInterface() {propertyValues {Curve {keys [<at 0 0.3 RCIM_Cubic>,<at 1 1>]}}}))
    ScaleA=true
    ScaleRGB=true
   )
  ]
  rendererClasses [&NiagaraSpriteRendererProperties]
  renderers [{rendererClass=&NiagaraSpriteRendererProperties}]
 }
 Feathers {
  emitterUpdateScript [
   &EmitterState(
    Inactive Response=::NewEnumerator1
    Life Cycle Mode=::NewEnumerator1
    Loop Behavior=::NewEnumerator1
    Loop Delay=0.4
    Loop Duration=2
    UseLoopDelay=false
   )
   &SpawnBurst_Instantaneous(Spawn Count=5, Spawn Time=0.04, Use Spawn Probability=false, UseLoopCountLimit=false)
  ]
  particleSpawnScript [
   &InitializeParticle(
    Color=&MergeHelper_ColorSwitch(
     0=<rgba 0.0118 0.0108 0.0273 1>
     1=<rgba 0.1198 0.0038 0.0037 1>
     2=?
     3=?
     Index=&RandomRangeInt(Minimum=0, Maximum=1)
     Num=::NewEnumerator0
    )
    Lifetime Max=1
    Lifetime Min=0.6
    Lifetime Mode=::NewEnumerator1
    Sprite Rotation Mode=::NewEnumerator2
    Sprite Size=<10 18>
    Sprite Size Mode=::NewEnumerator4
    UsePositionOffset=true
    Write Material Random=true
   )
   &CameraOffset(Camera Offset Amount=30)
   &AddVelocity(
    Rotation Coordinate Space=::World
    Velocity=&RandomRangeVector(Minimum=<-34 -34 0>, Maximum=<34 34 50>)
    Velocity Speed Scale=2
   ) {module=AddVelocity001}
   &ShapeLocation(MaskSpawnGroup=false, Override Seed=false, Sphere Radius=20, UseOwnerScale=true)
   &SubUVAnimation(
    Evaluation Type=::NewEnumerator0
    "Sprite Renderer "=DataInterface() {propertyValues {SpriteRenderer=&self:Feathers_0.NiagaraSpriteRendererProperties_0}}
    SubUV Animation Mode:ENiagara_SubUVLookupModeV2=::NewEnumerator2
    UseEndFrame=false
    UseStartFrame=false
   ) {module=SubUVAnimation001}
  ]
  particleUpdateScript [
   &ParticleState()
   &ScaleVelocity(Velocity Scale=&VectorFromCurve(VectorCurve=DataInterface() {propertyValues {
       XCurve {keys [{arriveTangent=-0.8665, interpMode=RCIM_Cubic, leaveTangent=-0.8665, tangentMode=RCTM_User, time=0, value=1}, <at 1 0.3>]}
       YCurve {keys [{arriveTangent=-0.8665, interpMode=RCIM_Cubic, leaveTangent=-0.8665, tangentMode=RCTM_User, time=0, value=1}, <at 1 0.3>]}
       ZCurve {keys [{arriveTangent=-0.8665, interpMode=RCIM_Cubic, leaveTangent=-0.8665, tangentMode=RCTM_User, time=0, value=1}, <at 1 0.3>]}
      }}))
   &SpriteRotationRate(Rotation Rate=&RandomRangeFloat(Minimum=-250, Maximum=250))
   &ScaleSpriteSize(
    Scale Sprite Size Mode=::NewEnumerator0
    Uniform Scale Factor=&FloatFromCurve(
     FloatCurve=DataInterface() {propertyValues {Curve {
        defaultValue=1
        keys [
         {arriveTangent=4.1785, interpMode=RCIM_Cubic, leaveTangent=4.1785, tangentMode=RCTM_User, time=0, value=1}
         {arriveTangent=1.1179, interpMode=RCIM_Cubic, leaveTangent=1.1179, tangentMode=RCTM_User, time=0.2042, value=1.6}
         {arriveTangent=0.0224, interpMode=RCIM_Cubic, leaveTangent=0.0224, tangentMode=RCTM_User, time=1, value=2}
        ]
       }}}
     Scale Curve=0.2
    )
   )
   &ScaleColor(
    Scale Alpha=&Multiply_Float(
     A=&RandomRangeFloat(Minimum=0.5, Maximum=1)
     B=&FloatFromCurve(
      FloatCurve=DataInterface() {propertyValues {Curve {keys [
          {arriveTangent=3.7485, interpMode=RCIM_Cubic, leaveTangent=3.7485, tangentMode=RCTM_User, time=0, value=0.35}
          {interpMode=RCIM_Cubic, tangentMode=RCTM_User, time=0.5, value=1}
          {tangentMode=RCTM_User, time=1, value=0}
         ]}}}
      Scale Curve=2
     )
    )
    ScaleA=true
    ScaleRGB=true
   )
   &GravityForce()
   &SolveForcesAndVelocity(Clamp Velocity=false, Limit Acceleration=false)
  ]
  rendererClasses [&NiagaraSpriteRendererProperties]
  renderers [{rendererClass=&NiagaraSpriteRendererProperties}]
 }
 Flash {
  emitterUpdateScript [
   &EmitterState(Inactive Response=::NewEnumerator1, Life Cycle Mode=::NewEnumerator1, Loop Behavior=::NewEnumerator1, UseLoopDelay=false)
   &SpawnBurst_Instantaneous(Spawn Count=3, Spawn Time=0.015, Use Spawn Probability=false, UseLoopCountLimit=false)
  ]
  particleSpawnScript [
   &FN_SetSpawnGroup(Number of Spawn Groups=3)
   &InitializeParticle(
    Color=<rgba 0.15 0.15 0.15 1.5>
    Lifetime=0.28
    Mesh Scale=<0.08 0.08 0.08>
    Mesh Scale Mode=::NewEnumerator4
    Position Offset=<0 0 10>
    Position Offset Coordinate Space=::Local
    Sprite Size=&MergeHelper_Vector2Switch(0=<190 140>, 1=<65 180>, 2=?, 3=?, 4=?, Index=&SpawnGroup(), Num=::NewEnumerator0)
    Sprite Size Mode=::NewEnumerator4
    UsePositionOffset=true
    Write Material Random=true
   )
   &InitialMeshOrientation(Mesh Orientation Mode=::NewEnumerator2, Orientation Vector=<0 0 1>, Rotation=<0 0.25 0>, Use Rotation Vector=true)
   &CameraOffset(Camera Offset Amount=-7)
   &AddVelocity(Velocity=<0 0 1>)
   &SpriteFacingAndAlignment(
    Sprite Alignment=&MergeHelper_VectorSwitch(0=<0 1 1>, 1=<1 0 1>, 2=?, 3=?, Index=&SpawnGroup(), Num=::NewEnumerator0)
    Sprite Facing=&MergeHelper_VectorSwitch(0=<0 1 0>, 1=<1 0 0>, 2=?, 3=?, Index=&SpawnGroup(), Num=::NewEnumerator0)
    WriteAlignment=true
    WriteFacing=true
   )
   /SetVariables(Particles.VisibilityTag=&SpawnGroup()) {isSetParameters=true}
  ]
  particleUpdateScript [
   &ParticleState()
   &ScaleMeshSize(Scale Factor=&Multiply_Vector(
     A=&ScaleFloatByCurve(
      InitialValue:Vector3f=<1 1 1>
      ScaleCurve:NiagaraDataInterfaceVectorCurve=DataInterface() {propertyValues {
        XCurve {keys [<at 0 0.25>,<at 0.1 1>,<at 0.9 1>,<at 1 0>]}
        YCurve {keys [<at 0 1.25>,<at 0.1 2.25>,<at 0.9 2.25>,<at 1 0>]}
        ZCurve {keys [<at 0 1.25>,<at 0.1 2.25>,<at 0.9 2.25>,<at 1 0>]}
       }}
      CurveIndex:NiagaraFloat=$Particles.NormalizedAge
     )
     B=&Divide_Vector(A:Vector3f=<1 1 1>, B:Vector3f=$Engine.Owner.Scale)
    ))
   &ScaleSpriteSize(
    Uniform Curve Scale=1.3
    Uniform Curve Sprite Scale=DataInterface() {propertyValues {Curve {keys [<at 0 0 RCIM_Cubic>,<at 1 1 RCIM_Cubic>]}}}
   )
   /SetVariables(Particles.Size002:Vector2f=&Multiply_Vector2D(A=$Particles.SpriteSize, B=<0.8 0.8>)) {isSetParameters=true}
   /SetVariables(Particles.Size003:Vector2f=&Multiply_Vector2D(A=<150 50>, B=<1 3>)) {isSetParameters=true}
   &ScaleColor(
    Scale Alpha=&Multiply_Float(
     A=&FloatFromCurve(
      FloatCurve=DataInterface() {propertyValues {Curve {keys [{arriveTangent=4.4475, interpMode=RCIM_Cubic, time=0, value=0}, <at 0.3788 1.2 RCIM_Cubic>, <at 1 0>]}}}
      Scale Curve=&FNTimeOfDayScalar(Night Scale=0.6, Morning Scale=0.7, Day Scale=0.8, Evening Scale=0.6)
     )
     B=&FN_FrontEndFloatModifier(Gameplay=1.2, Front End=0.3)
    )
    Scale RGB=&MakeVectorFromLinearColorRGB(Color=&ColorFromCurve(ColorCurve=DataInterface() {propertyValues {
        AlphaCurve {keys []}
        BlueCurve {keys [<at 0.7887 0 RCIM_Cubic>,<at 0.9801 0.0012 RCIM_Cubic>]}
        GreenCurve {keys [<at 0.7887 0.0063 RCIM_Cubic>,<at 0.9801 0 RCIM_Cubic>]}
        RedCurve {keys [<at 0.7887 0.4375 RCIM_Cubic>,<at 0.9801 0.1562 RCIM_Cubic>]}
       }}))
    ScaleA=true
    ScaleRGB=true
   )
   /SetVariables(Particles.color002:LinearColor=&MakeLinearColorFromVectorAndFloat(
     "Vector (RGB)"=&MakeVectorFromLinearColorRGB(Color=&Multiply_LinearColorByFloat(
       Linear Color=<rgba 15 0.2969 0.5834 1>
       Float=&Multiply_Float(
        A=&FN_FrontEndFloatModifier(Gameplay=1, Front End=0.8)
        B=&FNTimeOfDayScalar(Night Scale=0.5, Morning Scale=0.86, Evening Scale=0.77)
       )
      ))
     "Float (Alpha)"=&FloatFromCurve(
      FloatCurve=DataInterface() {propertyValues {Curve {keys [<at 0 0 RCIM_Cubic>,<at 0.2015 2 RCIM_Cubic>,<at 1 0>]}}}
      Scale Curve=&FN_FrontEndFloatModifier(Gameplay=1, Front End=0.4)
     )
    )) {isSetParameters=true}
   /SetVariables(Particles.Color003:LinearColor=&MakeLinearColorFromVectorAndFloat(
     "Vector (RGB)"=&MakeVectorFromLinearColorRGB(Color=<rgba 1 0.04 0.05 1>)
     "Float (Alpha)"=&Multiply_Float(
      A=&FloatFromCurve(
       FloatCurve=DataInterface() {propertyValues {Curve {keys [<at 0 0.8>,<at 0.1 1>,<at 0.2 0.6>,<at 0.3 1>,<at 0.4 0.7>,<at 0.5 0.95>,<at 0.6 0.55>,<at 0.7 1>,<at 0.8 0.55>,<at 0.9 0.8>,<at 1 0>]}}}
       Scale Curve=&FNTimeOfDayScalar(Night Scale=0.2, Morning Scale=0.1, Day Scale=0.4, Evening Scale=0.2)
      )
      B=&Multiply_Float(
       A=&FN_FrontEndFloatModifier(Gameplay=0.5, Front End=0.15)
       B=&FN_PickFloatDependOnPlatform(TargetPlatformActive=1, TargetPlatformNoActive=0.4)
      )
     )
    )) {isSetParameters=true}
   /SetVariables(Particles.MeshColor:LinearColor=&MakeLinearColorFromVectorAndFloat(
     "Vector (RGB)"=&MakeVectorFromLinearColorRGB(Color=&ColorFromCurve(ColorCurve=DataInterface() {propertyValues {
         AlphaCurve {keys []}
         BlueCurve {keys [
           {arriveTangent=-1.8762, interpMode=RCIM_Cubic, leaveTangent=-1.8762, tangentMode=RCTM_User, time=0.1298, value=0.3256}
           {interpMode=RCIM_Cubic, leaveTangent=0.2132, time=0.3441, value=0.1503}
          ]}
         GreenCurve {keys [
           {arriveTangent=-0.4331, interpMode=RCIM_Cubic, leaveTangent=-0.4331, tangentMode=RCTM_User, time=0.1298, value=0.2261}
           {interpMode=RCIM_Cubic, leaveTangent=2.0123, time=0.3441, value=0.1146}
          ]}
         RedCurve {keys [
           {arriveTangent=-4.577, interpMode=RCIM_Cubic, leaveTangent=-4.577, tangentMode=RCTM_User, time=0.1298, value=2}
           {interpMode=RCIM_Cubic, leaveTangent=-2.7231, time=0.3441, value=2}
          ]}
        }}))
     "Float (Alpha)"=&Multiply_Float(
      A=&FloatFromCurve(
       FloatCurve=DataInterface() {propertyValues {Curve {keys [
           {arriveTangent=-0.9, time=0, value=0}
           {arriveTangent=-0.0998, leaveTangent=-0.0998, time=0.05, value=1}
           {arriveTangent=-1.0872, leaveTangent=-1.0872, time=0.1804, value=0.7734}
           <at 0.2985 0.937>
           <at 0.4531 0.6834>
           <at 0.5813 0.9005>
           <at 0.7511 0.6392>
           <at 0.8564 0.8>
           {arriveTangent=-1.5912, leaveTangent=-1.5912, time=1, value=0}
          ]}}}
       Scale Curve=&FN_FrontEndFloatModifier(Gameplay=1, Front End=0.5)
      )
      B=&FNTimeOfDayScalar(Night Scale=0.7, Morning Scale=0.5, Evening Scale=0.6)
     )
    )) {isSetParameters=true}
   &FN_Light_Attributes(
    Color=&ColorFromCurve(ColorCurve=DataInterface() {propertyValues {
       AlphaCurve {keys []}
       BlueCurve {keys [<at 0 0 RCIM_Cubic>,<at 0.0858 0>,<at 0.1804 0>,<at 0.9 0.0729>,<at 1 0 RCIM_Cubic>]}
       GreenCurve {keys [<at 0 0 RCIM_Cubic>,<at 0.0858 0>,<at 0.1804 0>,<at 0.9 0.107>,<at 1 0 RCIM_Cubic>]}
       RedCurve {keys [<at 0 0 RCIM_Cubic>,<at 0.0858 30>,<at 0.1804 10>,<at 0.9 2>,<at 1 0 RCIM_Cubic>]}
      }})
    Create Color Attribute=true
    Create Exponent Attribute=true
    Create Position Attribute=true
    Create Radius Attribute=true
    Exponent=1
    Radius=130
    Scale Color=&FloatFromCurve(
     FloatCurve=DataInterface() {propertyValues {Curve {keys [<at 0 0.5 RCIM_Cubic>, {arriveTangent=-0.5, interpMode=RCIM_Cubic, leaveTangent=-0.5, time=0.4987, value=1}, <at 1 0>]}}}
     Scale Curve=&Multiply_Float(A=5, B=&FNTimeOfDayScalar(Night Scale=0.4, Morning Scale=0.6, Day Scale=0.8, Evening Scale=0.5))
    )
   )
   &SolveForcesAndVelocity(Clamp Velocity=false, Limit Acceleration=false)
  ]
  rendererClasses [&NiagaraSpriteRendererProperties, &NiagaraLightRendererProperties]
  renderers [
   {rendererClass=&NiagaraSpriteRendererProperties}
   {rendererClass=&NiagaraSpriteRendererProperties, rendererIndex=1}
   {rendererClass=&NiagaraLightRendererProperties, rendererIndex=2}
   {rendererClass=&NiagaraSpriteRendererProperties, rendererIndex=3}
   {rendererClass=&NiagaraSpriteRendererProperties, rendererIndex=4}
  ]
 }
 SmokeDetailed {
  emitterUpdateScript [
   &EmitterState(Life Cycle Mode=::NewEnumerator1, Loop Behavior=::NewEnumerator1, Loop Delay=0.4, Loop Duration=3.5, UseLoopDelay=false)
   &SpawnBurst_Instantaneous(Spawn Count=6, Spawn Time=0.05, Use Spawn Probability=false, UseLoopCountLimit=false)
  ]
  particleSpawnScript [
   &InitializeParticle(
    Color=&MergeHelper_ColorSwitch(
     0=<rgba 0.0118 0.0108 0.0273 1>
     1=<rgba 0.1198 0.0038 0.0037 1>
     2=?
     3=?
     Index=&RandomRangeInt(Minimum=0, Maximum=1)
     Num=::NewEnumerator0
    )
    Lifetime Max=0.65
    Lifetime Min=0.4
    Lifetime Mode=::NewEnumerator1
    Position Offset=<0 0 5>
    Position Offset Coordinate Space=::World
    Sprite Rotation Mode=::NewEnumerator2
    Sprite Size Max=<100 100>
    Sprite Size Min=<80 80>
    Sprite Size Mode=::NewEnumerator2
    UsePositionOffset=true
    Write Material Random=true
   )
   &CameraOffset(Camera Offset Amount=30)
   &AddVelocity(Velocity=&RandomRangeVector(Minimum=<-10 -10 40>, Maximum=<10 10 90>), Velocity Speed Scale=1.5) {module=AddVelocity001}
   &ShapeLocation(MaskSpawnGroup=false, Override Seed=false, Sphere Radius=12, UseOwnerScale=true)
  ]
  particleUpdateScript [
   &ParticleState()
   &ScaleVelocity(Velocity Scale=&VectorFromCurve(VectorCurve=DataInterface() {propertyValues {
       XCurve {keys [{arriveTangent=-1.756, interpMode=RCIM_Cubic, leaveTangent=-1.756, tangentMode=RCTM_User, time=0, value=1}, <at 1 0>]}
       YCurve {keys [{arriveTangent=-1.756, interpMode=RCIM_Cubic, leaveTangent=-1.756, tangentMode=RCTM_User, time=0, value=1}, <at 1 0>]}
       ZCurve {keys [{arriveTangent=-1.756, interpMode=RCIM_Cubic, leaveTangent=-1.756, tangentMode=RCTM_User, time=0, value=1}, <at 1 0>]}
      }}))
   &SpriteRotationRate(Rotation Rate=&RandomRangeFloat(Minimum=-50, Maximum=50))
   &ScaleSpriteSize(
    Scale Sprite Size Mode=::NewEnumerator0
    Uniform Scale Factor=&FloatFromCurve(
     FloatCurve=DataInterface() {propertyValues {Curve {
        defaultValue=1
        keys [
         {arriveTangent=4.1785, interpMode=RCIM_Cubic, leaveTangent=4.1785, tangentMode=RCTM_User, time=0, value=1}
         {arriveTangent=1.1179, interpMode=RCIM_Cubic, leaveTangent=1.1179, tangentMode=RCTM_User, time=0.2042, value=1.6}
         {arriveTangent=0.0224, interpMode=RCIM_Cubic, leaveTangent=0.0224, tangentMode=RCTM_User, time=1, value=2}
        ]
       }}}
     Scale Curve=0.25
    )
   )
   &SubUVAnimation(
    End Frame Range Override=30
    "Sprite Renderer "=DataInterface() {propertyValues {SpriteRenderer=&self:SmokeDetailed_0.NiagaraSpriteRendererProperties_0}}
    Start Frame Range Override=&RandomRangeInt(Minimum=0, Maximum=5)
    UseEndFrame=true
    UseStartFrame=true
   )
   &ScaleColor(
    Scale Alpha=&Multiply_Float(
     A=&RandomRangeFloat(Minimum=0.5, Maximum=1)
     B=&FloatFromCurve(
      FloatCurve=DataInterface() {propertyValues {Curve {keys [
          {arriveTangent=3.7485, interpMode=RCIM_Cubic, leaveTangent=3.7485, tangentMode=RCTM_User, time=0, value=0.35}
          {interpMode=RCIM_Cubic, tangentMode=RCTM_User, time=0.5, value=1}
          {tangentMode=RCTM_User, time=1, value=0}
         ]}}}
      Scale Curve=1.4
     )
    )
    ScaleA=true
    ScaleRGB=true
   )
   &SolveForcesAndVelocity(Clamp Velocity=false, Limit Acceleration=false)
  ]
  rendererClasses [&NiagaraSpriteRendererProperties]
  renderers [{rendererClass=&NiagaraSpriteRendererProperties}]
 }
 Spark {
  emitterUpdateScript [
   &EmitterState(Life Cycle Mode=::NewEnumerator1, Loop Behavior=::NewEnumerator1, Loop Delay=0.05, Loop Duration=1, UseLoopDelay=true)
   &SpawnBurst_Instantaneous(Spawn Count=7, Use Spawn Probability=false)
  ]
  particleSpawnScript [
   &InitializeParticle(
    Color=&MergeHelper_ColorSwitch(
     0=<rgba 1 1 1 1>
     1=<rgba 0.026 0 0 1.3>
     2=?
     3=?
     Index=&RandomRangeInt(Minimum=0, Maximum=1)
     Num=::NewEnumerator0
    )
    Lifetime Max=1
    Lifetime Min=0.7
    Lifetime Mode=::NewEnumerator1
    Position Offset Coordinate Space=::World
    Sprite Size Max=<3 8>
    Sprite Size Min=<2.5 3.5>
    Sprite Size Mode=::NewEnumerator2
    UsePositionOffset=true
    Write Material Random=true
   )
   &AddVelocityFromPoint(Module.Use Falloff Distance=false, Velocity Strength=&RandomRangeFloat(Minimum=100, Maximum=250))
   &ShapeLocation(
    MaskSpawnGroup=false
    Offset=<0 0 8>
    Offset Coordinate Space=::World
    Offset Mode=::NewEnumerator0
    Override Seed=false
    Sphere Radius=11
    UseOwnerScale=true
   )
  ]
  particleUpdateScript [
   &ParticleState()
   &ScaleColor(
    Linear Color Curve=DataInterface() {propertyValues {AlphaCurve {keys [<at 0.0074 1>,<at 0.8905 1>]}, BlueCurve {keys [<at 0 0.0455>]}, GreenCurve {keys [<at 0 0>]}, RedCurve {keys [<at 0 6>]}}}
    Scale Mode=::NewEnumerator2
   )
   &ScaleColor(
    Scale Alpha=&FNTimeOfDayScalar(Night Scale=0.5, Morning Scale=0.3, Evening Scale=0.3)
    Scale RGB=&VectorFromFloat(Value=1)
    ScaleA=true
    ScaleRGB=true
   ) {module=ScaleColor001}
   &Drag(Drag=&RandomRangeFloat(Minimum=3, Maximum=6), Use Linear Drag=true, Use Rotational Drag=false)
   &CurlNoiseForce(Noise Strength=&RandomRangeFloat(Minimum=500, Maximum=700), Pan Noise=true)
   &GravityForce(Gravity=&RandomRangeVector(Minimum=<0 0 0>, Maximum=<0 0 -100>))
   &SolveForcesAndVelocity(Clamp Velocity=false, Limit Acceleration=false)
   &ScaleSpriteSizeBySpeed(Max Scale Factor=<1 5>, Min Scale Factor=<1 0.05>)
   &ScaleSpriteSize(Scale Factor=&MakeVector2D(
     X=&FloatFromCurve(FloatCurve=DataInterface() {propertyValues {Curve {keys [<at 0 0.5>,<at 0.1 1>,<at 1 0.5>]}}}, Scale Curve=0.9)
     Y=&FloatFromCurve(FloatCurve=DataInterface() {propertyValues {Curve {keys [<at 0 1>,<at 1 0>]}}}, Scale Curve=0.9)
    ))
  ]
  rendererClasses [&NiagaraSpriteRendererProperties]
  renderers [{rendererClass=&NiagaraSpriteRendererProperties}]
 }
}
summary {
 systemName=NS_Emote_MistRaven_Book_MainClose
 emitters [
  {emitterName=Feathers, rendererClasses [&NiagaraSpriteRendererProperties]}
  {emitterName=Spark, rendererClasses [&NiagaraSpriteRendererProperties]}
  {emitterName=CapsuleGlow, rendererClasses [&NiagaraSpriteRendererProperties]}
  {emitterName=Flash, rendererClasses [&NiagaraSpriteRendererProperties, &NiagaraLightRendererProperties]}
  {emitterName=SmokeDetailed, rendererClasses [&NiagaraSpriteRendererProperties]}
 ]
 userVariables [
  User.bIsFrontEnd {type=NiagaraBool}
  User.bIsFrontEndPreview {type=NiagaraBool}
 ]
}
